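# Control Rig for the M_LRG_RadiumFox_Flame character body that applies procedural corrective deformation, notably to shoulder pads and thigh pads.
# A shared DeformRig function and fixed bandage and pad offsets run first, then sequenced per-region chains.
# Spherical pose readers, remapped bone distances, relative transforms and rotations, often smoothed, weight transform offsets and filtered parent constraints.
# Special fixes cover the right pinky, a thigh-derived skydive weight and a Lunch Emote neck adjustment.
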ControlRigBlueprint /BRCosmetics/Characters/Player/Male/Large/Bodies/M_LRG_RadiumFox_Flame/Meshes/M_LRG_RadiumFox_Flame_CtrlRig
graphs [
 ConstructionGraph {nodes [RigUnit_PrepareForExecution = &RigUnit_PrepareForExecution()]}
 Multiply_1_1_ContainedGraph {nodes [
   n1 = RigVMFunctionEntryNode()
   RigVMFunctionReturnNode(Result=&RigVMFunction_MathDoubleMul(B=@n1.C, A=&RigVMFunction_MathDoubleMul(A=@n1.A, B=@n1.B).Result).Result)
  ]}
 Multiply_1_1_ContainedGraph {nodes [
   n1 = RigVMFunctionEntryNode()
   RigVMFunctionReturnNode(Result=&RigVMFunction_MathDoubleMul(B=@n1.C, A=&RigVMFunction_MathDoubleMul(A=@n1.A, B=@n1.B).Result).Result)
  ]}
 RigVMFunctionLibrary {links [], nodes []}
 RigVMFunction_Sequence_11_1_ContainedGraph {nodes [
   n1 = RigVMFunctionEntryNode()
   n2 = &RigVMFunction_Sequence(ExecuteContext=@n1.ExecuteContext)
   n3 = &RigVMFunction_Sequence(ExecuteContext=@n2.B)
   n4 = &RigVMFunction_Sequence(ExecuteContext=@n3.B)
   RigVMFunctionReturnNode(A=@n2.A, B=@n3.A, C=@n4.A, D=@n4.B)
  ]}
 RigVMFunction_Sequence_1_1_ContainedGraph {nodes [
   n1 = RigVMFunctionEntryNode()
   n2 = &RigVMFunction_Sequence(ExecuteContext=@n1.ExecuteContext)
   n3 = &RigVMFunction_Sequence(ExecuteContext=@n2.B)
   n4 = &RigVMFunction_Sequence(ExecuteContext=@n3.B)
   n5 = &RigVMFunction_Sequence(ExecuteContext=@n4.B)
   n6 = &RigVMFunction_Sequence(ExecuteContext=@n5.B)
   n7 = &RigVMFunction_Sequence(ExecuteContext=@n6.B)
   n8 = &RigVMFunction_Sequence(ExecuteContext=@n7.B)
   n9 = &RigVMFunction_Sequence(ExecuteContext=@n8.B)
   RigVMFunctionReturnNode(A=@n2.A, B=@n3.A, C=@n4.A, D=@n5.A, E=@n6.A, F=@n7.A, G=@n8.A, H=@n9.A, I=@n9.B)
  ]}
 RigVMModel {nodes [
   RigUnit_BeginExecution = &RigUnit_BeginExecution()
   FN_DeformRig_FNC = RigVMFunctionReferenceNode(
    DeformRig_ Curve="(deformRig_l_upperarm_up_alpha_72_#1=true,deformRig_r_upperarm_up_alpha_73_#2=true,deformRig_l_upperarm_fr_alpha_74_#3=true,deformRig_r_upperarm_fr_alpha_75_CB51B80C4 ...(truncated)"
    DeformRig_Bool="(upperarm_1_#4=True,elbow_3_#5=True,wrist_7_#6=True,groin_8_#7=True,glute_17_#8=True,knee_9_40C408324CA261 ...(truncated)"
    DeformRig_Debug=(upperarm_debug_18_#4=False,wrist_debug_20_#6=False,lat_debug_26_#9=False,groinGlute_debug_27_#7=False)
    DeformRig_Value="(upperarm_up_18_#10=(Rotation=(X=0,Y=0,Z=0,W=1),Translation=(X=0,Y=0,Z=8),Scale3D=(X=0.7,Y=1,Z=1)),upperarm_fwd_fr_19_#11=(Rotation=(X=0,Y=0,Z=0,W=1),Translation=(X=0,Y=10,Z=0),Scale3D=(X=0.5,Y=1,Z= ...(truncated)"
    ExecuteContext=@RigUnit_BeginExecution.ExecutePin
   )
   > &RigUnit_ModifyTransforms(
    ItemToModify="((Item=(Type=Bone,Name=\"deform_shoulderpad_l_base\"),Transform=(Rotation=(X=0.0349,Y=0,Z=0,W=0.9994),Translation=(X=1,Y=1,Z=0),Scale3D=(X=1,Y=1,Z=1))),(Item=(Type=Bone,Name=\"deform_shoulderpad_r_base\"),Transform=(Rotation=(X=0.0349,Y=0,Z=0,W=0.9994),Transla ...(truncated)"
    Weight=1
    WeightMinimum=0
   )
   > &RigUnit_ModifyTransforms(
    ItemToModify=((Item=(Type=Bone,Name="deform_bandage_root"),Transform=(Rotation=(X=0.2164,Y=0,Z=0,W=0.9763),Translation=(X=0,Y=-0.2,Z=0),Scale3D=(X=1,Y=1.03,Z=1))))
    Weight=1
    WeightMinimum=0
   )
   > RigVMFunction_Sequence_1_1 = RigVMAggregateNode()
   RigVMFunction_Sequence_11 = RigVMAggregateNode(ExecuteContext=@RigVMFunction_Sequence_1_1.A)
   &RigUnit_ModifyTransforms(
    ItemToModify=((Item=(Type=Bone,Name="clavicle_l"),Transform=(Rotation=(X=0,Y=0,Z=0,W=1),Translation=(X=0,Y=0,Z=0),Scale3D=(X=1,Y=1,Z=1))))
    Weight=1
    WeightMinimum=0
    ExecutePin=@RigVMFunction_Sequence_11.A
   )
   > &RigUnit_ModifyTransforms(
    ItemToModify=((Item=(Type=Bone,Name="upperarm_l"),Transform=(Rotation=(X=0,Y=0,Z=0,W=1),Translation=(X=0,Y=0,Z=0),Scale3D=(X=1,Y=1,Z=1))))
    Weight=1
    WeightMinimum=0
   )
   > &RigUnit_ModifyTransforms(
    ItemToModify=((Item=(Type=Bone,Name="lowerarm_l"),Transform=(Rotation=(X=0,Y=0,Z=0,W=1),Translation=(X=0,Y=0,Z=0),Scale3D=(X=1,Y=1,Z=1))))
    Weight=1
    WeightMinimum=0
   )
   RigVMFunction_MathVectorDistance = &RigVMFunction_MathVectorDistance(
    A=&RigUnit_GetTransform(Item=(Type=Bone,Name="deform_shoulderpad_target_up_l"), bInitial=False).Transform.Translation
    B=&RigUnit_GetTransform(Item=(Type=Bone,Name="head"), bInitial=False).Transform.Translation
   )
   n1 = &RigVMFunction_AlphaInterp(
    InterpSpeedDecreasing=30
    InterpSpeedIncreasing=30
    Scale=1
    bClampResult=true
    Value=&RigVMFunction_MathDoubleRemap(SourceMaximum=8, SourceMinimum=25, TargetMaximum=1, TargetMinimum=0, Value=@RigVMFunction_MathVectorDistance.Result).Result
   )
   > &RigUnit_ModifyTransforms(
    ItemToModify=((Item=(Type=Bone,Name="deform_shoulderpad_l"),Transform=(Rotation=(X=0,Y=0.2334,Z=0,W=0.9724),Translation=(X=4,Y=0,Z=0),Scale3D=(X=0.9,Y=1,Z=0.9))))
    WeightMinimum=0
    Weight=@n1.Result
   )
   n2 = &RigUnit_GetRelativeTransformForItem(
    Child=(Type=Bone,Name="deform_shoulderpad_target_side_l")
    Parent=(Type=Bone,Name="upperarm_twist_02_l")
    bChildInitial=false
    bParentInitial=false
   )
   n3 = &RigVMFunction_AlphaInterp(
    InterpSpeedDecreasing=30
    InterpSpeedIncreasing=30
    Scale=1
    bClampResult=true
    Value=&RigVMFunction_MathDoubleRemap(SourceMaximum=-12, SourceMinimum=9, TargetMaximum=1, TargetMinimum=0, Value=@n2.RelativeTransform.Translation.Z).Result
   )
   > &RigUnit_ModifyTransforms(
    ItemToModify=((Item=(Type=Bone,Name="deform_shoulderpad_l"),Transform=(Rotation=(X=0,Y=-0.3827,Z=0,W=0.9239),Translation=(X=1,Y=0,Z=9),Scale3D=(X=0.9,Y=1,Z=0.9))))
    WeightMinimum=0
    Weight=@n3.Result
   )
   > Get Delta Ref_2 = RigVMFunctionReferenceNode(Item=(Type=Bone,Name="clavicle_l"))
   n4 = &RigVMFunction_AlphaInterp(
    InterpSpeedDecreasing=30
    InterpSpeedIncreasing=30
    Scale=1
    bClampResult=true
    Value=&RigVMFunction_MathDoubleRemap(SourceMaximum=30, SourceMinimum=0, TargetMaximum=1, TargetMinimum=0, Value=@"Get Delta Ref_2".Result.Rotation.Yaw).Result
   )
   > &RigUnit_ModifyTransforms(
    ItemToModify=((Item=(Type=Bone,Name="deform_shoulderpad_l_base_root"),Transform=(Rotation=(X=0,Y=0,Z=-0.1478,W=0.989),Translation=(X=0,Y=-1,Z=0),Scale3D=(X=1,Y=1,Z=1))))
    WeightMinimum=0
    Weight=@n4.Result
   )
   n5 = &RigVMFunction_AlphaInterp(
    InterpSpeedDecreasing=30
    InterpSpeedIncreasing=30
    Scale=1
    bClampResult=true
    Value=&RigVMFunction_MathDoubleRemap(SourceMaximum=-30, SourceMinimum=0, TargetMaximum=1, TargetMinimum=0, Value=@"Get Delta Ref_2".Result.Rotation.Yaw).Result
   )
   > &RigUnit_ModifyTransforms(
    ItemToModify=((Item=(Type=Bone,Name="deform_shoulderpad_l_base_root"),Transform=(Rotation=(X=0,Y=0,Z=0.2164,W=0.9763),Translation=(X=2,Y=0,Z=0),Scale3D=(X=1,Y=1,Z=1))))
    WeightMinimum=0
    Weight=@n5.Result
   )
   n6 = &RigVMFunction_AlphaInterp(
    InterpSpeedDecreasing=30
    InterpSpeedIncreasing=30
    Scale=1
    bClampResult=true
    Value=&RigVMFunction_MathDoubleRemap(SourceMaximum=30, SourceMinimum=0, TargetMaximum=1, TargetMinimum=0, Value=@"Get Delta Ref_2".Result.Rotation.Pitch).Result
   )
   > &RigUnit_ModifyTransforms(
    ItemToModify=((Item=(Type=Bone,Name="deform_shoulderpad_l"),Transform=(Rotation=(X=0,Y=0.0436,Z=0,W=0.999),Translation=(X=0,Y=0,Z=0),Scale3D=(X=0.95,Y=0.95,Z=0.95))))
    WeightMinimum=0
    Weight=@n6.Result
   )
   n7 = &RigUnit_GetRelativeTransformForItem(
    Child=(Type=Bone,Name="deform_shoulderpad_target_side_l")
    Parent=(Type=Bone,Name="lowerarm_twist_02_l")
    bChildInitial=False
    bParentInitial=False
   )
   n8 = &RigVMFunction_AlphaInterp(
    InterpSpeedDecreasing=30
    InterpSpeedIncreasing=30
    Scale=1
    bClampResult=true
    Value=&RigVMFunction_MathDoubleRemap(SourceMaximum=1, SourceMinimum=-10, TargetMaximum=1, TargetMinimum=0, Value=@n7.RelativeTransform.Translation.X).Result
   )
   > &RigUnit_ModifyTransforms(
    ItemToModify=((Item=(Type=Bone,Name="deform_shoulderpad_l"),Transform=(Rotation=(X=0,Y=0.0436,Z=0,W=0.999),Translation=(X=-2,Y=0,Z=2),Scale3D=(X=0.95,Y=1,Z=0.95))))
    WeightMinimum=0
    Weight=@n8.Result
   )
   n9 = &RigUnit_GetRelativeTransformForItem(Child=(Type=Bone,Name="lowerarm_l"), Parent=(Type=Bone,Name="clavicle_l"), bChildInitial=False, bParentInitial=false)
   n10 = &RigVMFunction_AlphaInterp(
    InterpSpeedDecreasing=30
    InterpSpeedIncreasing=30
    Scale=1
    bClampResult=true
    Value=&RigVMFunction_MathDoubleRemap(SourceMaximum=25, SourceMinimum=0, TargetMaximum=1, TargetMinimum=0, Value=@n9.RelativeTransform.Translation.Y).Result
   )
   > &RigUnit_ModifyTransforms(
    ItemToModify=((Item=(Type=Bone,Name="deform_shoulderpad_l_base_root"),Transform=(Rotation=(X=0.2565,Y=-0.0534,Z=-0.0308,W=0.9646),Translation=(X=0,Y=1,Z=0),Scale3D=(X=1,Y=1,Z=1))))
    WeightMinimum=0
    Weight=@n10.Result
   )
   n11 = &RigVMFunction_AlphaInterp(
    InterpSpeedDecreasing=30
    InterpSpeedIncreasing=30
    Scale=1
    bClampResult=true
    Value=&RigVMFunction_MathDoubleRemap(SourceMaximum=-25, SourceMinimum=0, TargetMaximum=1, TargetMinimum=0, Value=@n9.RelativeTransform.Translation.Y).Result
   )
   > &RigUnit_ModifyTransforms(
    ItemToModify=((Item=(Type=Bone,Name="deform_shoulderpad_l_base_root"),Transform=(Rotation=(X=-0.2578,Y=-0.0226,Z=0.0842,W=0.9623),Translation=(X=0,Y=0,Z=0),Scale3D=(X=1,Y=1,Z=1))))
    WeightMinimum=0
    Weight=@n11.Result
   )
   > n12 = &RigUnit_SphericalPoseReader(
    ActiveRegionScaleFactors=(PositiveWidth=1,NegativeWidth=1,PositiveHeight=1,NegativeHeight=1)
    ActiveRegionSize=0.07
    Debug=(bDrawDebug=false,bDraw2D=False,bDrawLocalAxes=False,DebugScale=25,DebugSegments=20,DebugThickness=0.25)
    DriverAxis=(X=1,Y=0,Z=0)
    DriverItem=(Type=Bone,Name="upperarm_l")
    FalloffRegionScaleFactors=(PositiveWidth=1,NegativeWidth=1,PositiveHeight=1,NegativeHeight=1)
    FalloffSize=0.25
    FlipHeightScaling=False
    FlipWidthScaling=False
    OptionalParentItem=(Type=Bone,Name="clavicle_l")
    RotationOffset=(X=12,Y=12,Z=0)
   )
   > &RigUnit_ModifyTransforms(
    ItemToModify=((Item=(Type=Bone,Name="deform_shoulderpad_l"),Transform=(Rotation=(X=0.3409,Y=0.0473,Z=-0.1139,W=0.932),Translation=(X=-3.2797,Y=-2.8982,Z=0.0399),Scale3D=(X=1,Y=1,Z=1))))
    WeightMinimum=0
    Weight=@n12.OutputParam
   )
   > n13 = &RigUnit_SphericalPoseReader(
    ActiveRegionScaleFactors=(PositiveWidth=1,NegativeWidth=1,PositiveHeight=1,NegativeHeight=1)
    ActiveRegionSize=0.025
    Debug=(bDrawDebug=false,bDraw2D=False,bDrawLocalAxes=False,DebugScale=25,DebugSegments=20,DebugThickness=0.25)
    DriverAxis=(X=1,Y=0,Z=0)
    DriverItem=(Type=Bone,Name="lowerarm_l")
    FalloffRegionScaleFactors=(PositiveWidth=1,NegativeWidth=1,PositiveHeight=1,NegativeHeight=1)
    FalloffSize=0.2
    FlipHeightScaling=False
    FlipWidthScaling=False
    OptionalParentItem=(Type=Bone,Name="upperarm_l")
    RotationOffset=(X=0,Y=-95,Z=-5)
   )
   > n14 = &RigUnit_SphericalPoseReader(
    ActiveRegionScaleFactors=(PositiveWidth=1,NegativeWidth=1,PositiveHeight=1,NegativeHeight=1)
    ActiveRegionSize=0.15
    Debug=(bDrawDebug=false,bDraw2D=False,bDrawLocalAxes=False,DebugScale=25,DebugSegments=20,DebugThickness=0.25)
    DriverAxis=(X=1,Y=0,Z=0)
    DriverItem=(Type=Bone,Name="upperarm_l")
    FalloffRegionScaleFactors=(PositiveWidth=1,NegativeWidth=1,PositiveHeight=1,NegativeHeight=1)
    FalloffSize=0.3
    FlipHeightScaling=False
    FlipWidthScaling=False
    OptionalParentItem=(Type=Bone,Name="clavicle_l")
    RotationOffset=(X=0,Y=-35,Z=0)
   )
   n15 = &RigVMFunction_MathDoubleMul(B=1, A=&RigVMFunction_MathDoubleMul(B=@n13.OutputParam, A=@n14.OutputParam).Result)
   > &RigUnit_ParentConstraint(
    AdvancedSettings=(InterpolationType=Average,RotationOrderForFilter=XYZ)
    Child=(Type=Bone,Name="deform_shoulderpad_l_base_root")
    Filter=(TranslationFilter=(bX=false,bY=true,bZ=false),RotationFilter=(bX=false,bY=false,bZ=false),ScaleFilter=(bX=false,bY=false,bZ=false))
    Parents=((Item=(Type=Bone,Name="upperarm_l"),Weight=0.5),(Item=(Type=Bone,Name="lowerarm_l"),Weight=1))
    bMaintainOffset=true
    Weight=@n15.Result
   )
   > &RigUnit_ParentConstraint(
    AdvancedSettings=(InterpolationType=Average,RotationOrderForFilter=XYZ)
    Child=(Type=Bone,Name="deform_shoulderpad_l_base_root")
    Filter=(TranslationFilter=(bX=false,bY=false,bZ=false),RotationFilter=(bX=false,bY=true,bZ=true),ScaleFilter=(bX=false,bY=false,bZ=false))
    Parents=((Item=(Type=Bone,Name="clavicle_l"),Weight=0.8),(Item=(Type=Bone,Name="upperarm_l"),Weight=1))
    bMaintainOffset=false
    Weight=@n15.Result
   )
   > &RigUnit_ModifyTransforms(
    ItemToModify=((Item=(Type=Bone,Name="deform_shoulderpad_l_base_root"),Transform=(Rotation=(X=0.0014,Y=0.0261,Z=-0.0523,W=0.9983),Translation=(X=2,Y=0,Z=0),Scale3D=(X=1,Y=1,Z=1))))
    WeightMinimum=0
    Weight=@n15.Result
   )
   &RigUnit_ModifyTransforms(
    ItemToModify=((Item=(Type=Bone,Name="clavicle_r"),Transform=(Rotation=(X=0,Y=0,Z=0,W=1),Translation=(X=0,Y=0,Z=0),Scale3D=(X=1,Y=1,Z=1))))
    Weight=1
    WeightMinimum=0
    ExecutePin=@RigVMFunction_Sequence_11.B
   )
   > &RigUnit_ModifyTransforms(
    ItemToModify=((Item=(Type=Bone,Name="upperarm_r"),Transform=(Rotation=(X=0,Y=0,Z=0,W=1),Translation=(X=0,Y=0,Z=0),Scale3D=(X=1,Y=1,Z=1))))
    Weight=1
    WeightMinimum=0
   )
   > &RigUnit_ModifyTransforms(
    ItemToModify=((Item=(Type=Bone,Name="lowerarm_r"),Transform=(Rotation=(X=0,Y=0,Z=0,W=1),Translation=(X=0,Y=0,Z=0),Scale3D=(X=1,Y=1,Z=1))))
    Weight=1
    WeightMinimum=0
   )
   > &RigUnit_ParentConstraint(
    AdvancedSettings=(InterpolationType=Average,RotationOrderForFilter=XZY)
    Child=(Type=Null,Name="upperarm_twist_02_r_null")
    Filter=(TranslationFilter=(bX=true,bY=true,bZ=true),RotationFilter=(bX=false,bY=false,bZ=false),ScaleFilter=(bX=false,bY=false,bZ=false))
    Parents=((Item=(Type=Bone,Name="upperarm_twist_02_r"),Weight=1))
    Weight=1
    bMaintainOffset=false
   )
   > &RigUnit_ParentConstraint(
    AdvancedSettings=(InterpolationType=Average,RotationOrderForFilter=XZY)
    Child=(Type=Null,Name="lowerarm_r_null")
    Filter=(TranslationFilter=(bX=true,bY=true,bZ=true),RotationFilter=(bX=false,bY=false,bZ=false),ScaleFilter=(bX=false,bY=false,bZ=false))
    Parents=((Item=(Type=Bone,Name="lowerarm_r"),Weight=1))
    Weight=1
    bMaintainOffset=false
   )
   n16 = &RigUnit_GetTransform(Item=(Type=Bone,Name="head"), bInitial=False)
   n17 = &RigUnit_GetTransform(Item=(Type=Bone,Name="deform_shoulderpad_target_up_r"), bInitial=False)
   n18 = &RigVMFunction_AlphaInterp(
    InterpSpeedDecreasing=30
    InterpSpeedIncreasing=30
    Scale=1
    bClampResult=true
    Value=&RigVMFunction_MathDoubleRemap(
     SourceMaximum=8
     SourceMinimum=25
     TargetMaximum=1
     TargetMinimum=0
     Value=&RigVMFunction_MathVectorDistance(B=@n16.Transform.Translation, A=@n17.Transform.Translation).Result
    ).Result
   )
   > &RigUnit_ModifyTransforms(
    ItemToModify=((Item=(Type=Bone,Name="deform_shoulderpad_r"),Transform=(Rotation=(X=0,Y=0.2164,Z=0,W=0.9763),Translation=(X=-4,Y=0,Z=0),Scale3D=(X=0.9,Y=1,Z=0.9))))
    WeightMinimum=0
    Weight=@n18.Result
   )
   n19 = &RigUnit_GetRelativeTransformForItem(
    Child=(Type=Bone,Name="deform_shoulderpad_target_side_r")
    Parent=(Type=Bone,Name="upperarm_twist_02_r")
    bChildInitial=false
    bParentInitial=false
   )
   n20 = &RigVMFunction_AlphaInterp(
    InterpSpeedDecreasing=30
    InterpSpeedIncreasing=30
    Scale=1
    bClampResult=true
    Value=&RigVMFunction_MathDoubleRemap(SourceMaximum=12, SourceMinimum=-9, TargetMaximum=1, TargetMinimum=0, Value=@n19.RelativeTransform.Translation.Z).Result
   )
   > &RigUnit_ModifyTransforms(
    ItemToModify=((Item=(Type=Bone,Name="deform_shoulderpad_r"),Transform=(Rotation=(X=0,Y=-0.3827,Z=0,W=0.9239),Translation=(X=-1,Y=0,Z=-9),Scale3D=(X=0.9,Y=1,Z=0.9))))
    WeightMinimum=0
    Weight=@n20.Result
   )
   > Get Delta Ref_2_2 = RigVMFunctionReferenceNode(Item=(Type=Bone,Name="clavicle_r"))
   n21 = &RigVMFunction_AlphaInterp(
    InterpSpeedDecreasing=30
    InterpSpeedIncreasing=30
    Scale=1
    bClampResult=true
    Value=&RigVMFunction_MathDoubleRemap(SourceMaximum=30, SourceMinimum=0, TargetMaximum=1, TargetMinimum=0, Value=@"Get Delta Ref_2_2".Result.Rotation.Yaw).Result
   )
   > &RigUnit_ModifyTransforms(
    ItemToModify=((Item=(Type=Bone,Name="deform_shoulderpad_r_base_root"),Transform=(Rotation=(X=0,Y=0,Z=-0.1478,W=0.989),Translation=(X=0,Y=1,Z=0),Scale3D=(X=1,Y=1,Z=1))))
    WeightMinimum=0
    Weight=@n21.Result
   )
   n22 = &RigVMFunction_AlphaInterp(
    InterpSpeedDecreasing=30
    InterpSpeedIncreasing=30
    Scale=1
    bClampResult=true
    Value=&RigVMFunction_MathDoubleRemap(SourceMaximum=-30, SourceMinimum=0, TargetMaximum=1, TargetMinimum=0, Value=@"Get Delta Ref_2_2".Result.Rotation.Yaw).Result
   )
   > &RigUnit_ModifyTransforms(
    ItemToModify=((Item=(Type=Bone,Name="deform_shoulderpad_r_base_root"),Transform=(Rotation=(X=0,Y=0,Z=0.2164,W=0.9763),Translation=(X=-2,Y=0,Z=0),Scale3D=(X=1,Y=1,Z=1))))
    WeightMinimum=0
    Weight=@n22.Result
   )
   n23 = &RigVMFunction_AlphaInterp(
    InterpSpeedDecreasing=30
    InterpSpeedIncreasing=30
    Scale=1
    bClampResult=true
    Value=&RigVMFunction_MathDoubleRemap(SourceMaximum=30, SourceMinimum=0, TargetMaximum=1, TargetMinimum=0, Value=@"Get Delta Ref_2_2".Result.Rotation.Pitch).Result
   )
   > &RigUnit_ModifyTransforms(
    ItemToModify=((Item=(Type=Bone,Name="deform_shoulderpad_r"),Transform=(Rotation=(X=0,Y=0.0436,Z=0,W=0.999),Translation=(X=0,Y=0,Z=0),Scale3D=(X=0.95,Y=0.95,Z=0.95))))
    WeightMinimum=0
    Weight=@n23.Result
   )
   n24 = &RigUnit_GetRelativeTransformForItem(
    Child=(Type=Bone,Name="deform_shoulderpad_target_side_r")
    Parent=(Type=Bone,Name="lowerarm_twist_02_r")
    bChildInitial=False
    bParentInitial=False
   )
   n25 = &RigVMFunction_AlphaInterp(
    InterpSpeedDecreasing=30
    InterpSpeedIncreasing=30
    Scale=1
    bClampResult=true
    Value=&RigVMFunction_MathDoubleRemap(SourceMaximum=-1, SourceMinimum=10, TargetMaximum=1, TargetMinimum=0, Value=@n24.RelativeTransform.Translation.X).Result
   )
   > &RigUnit_ModifyTransforms(
    ItemToModify=((Item=(Type=Bone,Name="deform_shoulderpad_r"),Transform=(Rotation=(X=0,Y=0.0436,Z=0,W=0.999),Translation=(X=2,Y=0,Z=-2),Scale3D=(X=0.95,Y=1,Z=0.95))))
    WeightMinimum=0
    Weight=@n25.Result
   )
   n26 = &RigUnit_GetRelativeTransformForItem(Child=(Type=Bone,Name="lowerarm_r"), Parent=(Type=Bone,Name="clavicle_r"), bChildInitial=False, bParentInitial=False)
   n27 = &RigVMFunction_AlphaInterp(
    InterpSpeedDecreasing=30
    InterpSpeedIncreasing=30
    Scale=1
    bClampResult=true
    Value=&RigVMFunction_MathDoubleRemap(SourceMaximum=-25, SourceMinimum=0, TargetMaximum=1, TargetMinimum=0, Value=@n26.RelativeTransform.Translation.Y).Result
   )
   > &RigUnit_ModifyTransforms(
    ItemToModify=((Item=(Type=Bone,Name="deform_shoulderpad_r_base_root"),Transform=(Rotation=(X=0.2586,Y=-0.0113,Z=-0.0421,W=0.965),Translation=(X=0,Y=-1,Z=2.5),Scale3D=(X=1,Y=1,Z=1))))
    WeightMinimum=0
    Weight=@n27.Result
   )
   n28 = &RigVMFunction_AlphaInterp(
    InterpSpeedDecreasing=30
    InterpSpeedIncreasing=30
    Scale=1
    bClampResult=true
    Value=&RigVMFunction_MathDoubleRemap(SourceMaximum=25, SourceMinimum=0, TargetMaximum=1, TargetMinimum=0, Value=@n26.RelativeTransform.Translation.Y).Result
   )
   > &RigUnit_ModifyTransforms(
    ItemToModify=((Item=(Type=Bone,Name="deform_shoulderpad_r_base_root"),Transform=(Rotation=(X=-0.2578,Y=-0.0226,Z=0.0842,W=0.9623),Translation=(X=0,Y=0,Z=0),Scale3D=(X=1,Y=1,Z=1))))
    WeightMinimum=0
    Weight=@n28.Result
   )
   > n29 = &RigUnit_SphericalPoseReader(
    ActiveRegionScaleFactors=(PositiveWidth=1,NegativeWidth=1,PositiveHeight=1,NegativeHeight=1)
    ActiveRegionSize=0.07
    Debug=(bDrawDebug=false,bDraw2D=False,bDrawLocalAxes=False,DebugScale=25,DebugSegments=20,DebugThickness=0.25)
    DriverAxis=(X=-1,Y=0,Z=0)
    DriverItem=(Type=Bone,Name="upperarm_r")
    FalloffRegionScaleFactors=(PositiveWidth=1,NegativeWidth=1,PositiveHeight=1,NegativeHeight=1)
    FalloffSize=0.25
    FlipHeightScaling=False
    FlipWidthScaling=False
    OptionalParentItem=(Type=Bone,Name="clavicle_r")
    RotationOffset=(X=210,Y=36,Z=0)
   )
   > &RigUnit_ModifyTransforms(
    ItemToModify=((Item=(Type=Bone,Name="deform_shoulderpad_r"),Transform=(Rotation=(X=0.4241,Y=0.0026,Z=-0.0973,W=0.9004),Translation=(X=3.2797,Y=2.8983,Z=-0.0399),Scale3D=(X=1,Y=1,Z=1))))
    WeightMinimum=0
    Weight=@n29.OutputParam
   )
   > n30 = &RigUnit_SphericalPoseReader(
    ActiveRegionScaleFactors=(PositiveWidth=1,NegativeWidth=1,PositiveHeight=1,NegativeHeight=1)
    ActiveRegionSize=0.025
    Debug=(bDrawDebug=false,bDraw2D=False,bDrawLocalAxes=False,DebugScale=25,DebugSegments=20,DebugThickness=0.25)
    DriverAxis=(X=-1,Y=0,Z=0)
    DriverItem=(Type=Bone,Name="lowerarm_r")
    FalloffRegionScaleFactors=(PositiveWidth=1,NegativeWidth=1,PositiveHeight=1,NegativeHeight=1)
    FalloffSize=0.2
    FlipHeightScaling=False
    FlipWidthScaling=False
    OptionalParentItem=(Type=Bone,Name="upperarm_r")
    RotationOffset=(X=0,Y=85,Z=-10)
   )
   > n31 = &RigUnit_SphericalPoseReader(
    ActiveRegionScaleFactors=(PositiveWidth=1,NegativeWidth=1,PositiveHeight=1,NegativeHeight=1)
    ActiveRegionSize=0.15
    Debug=(bDrawDebug=false,bDraw2D=False,bDrawLocalAxes=False,DebugScale=25,DebugSegments=20,DebugThickness=0.25)
    DriverAxis=(X=-1,Y=0,Z=0)
    DriverItem=(Type=Bone,Name="upperarm_r")
    FalloffRegionScaleFactors=(PositiveWidth=1,NegativeWidth=1,PositiveHeight=1,NegativeHeight=1)
    FalloffSize=0.3
    FlipHeightScaling=False
    FlipWidthScaling=False
    OptionalParentItem=(Type=Bone,Name="clavicle_r")
    RotationOffset=(X=0,Y=145,Z=0)
   )
   n32 = &RigVMFunction_MathDoubleMul(B=1, A=&RigVMFunction_MathDoubleMul(A=@n31.OutputParam, B=@n30.OutputParam).Result)
   > &RigUnit_ParentConstraint(
    AdvancedSettings=(InterpolationType=Average,RotationOrderForFilter=XYZ)
    Child=(Type=Bone,Name="deform_shoulderpad_r_base_root")
    Filter=(TranslationFilter=(bX=false,bY=true,bZ=false),RotationFilter=(bX=false,bY=false,bZ=false),ScaleFilter=(bX=false,bY=false,bZ=false))
    Parents=((Item=(Type=Bone,Name="upperarm_r"),Weight=0.5),(Item=(Type=Bone,Name="lowerarm_r"),Weight=1))
    bMaintainOffset=true
    Weight=@n32.Result
   )
   > &RigUnit_ParentConstraint(
    AdvancedSettings=(InterpolationType=Average,RotationOrderForFilter=XYZ)
    Child=(Type=Bone,Name="deform_shoulderpad_r_base_root")
    Filter=(TranslationFilter=(bX=false,bY=false,bZ=false),RotationFilter=(bX=false,bY=true,bZ=true),ScaleFilter=(bX=false,bY=false,bZ=false))
    Parents=((Item=(Type=Bone,Name="clavicle_r"),Weight=0.8),(Item=(Type=Bone,Name="upperarm_r"),Weight=1))
    bMaintainOffset=false
    Weight=@n32.Result
   )
   > &RigUnit_ModifyTransforms(
    ItemToModify=((Item=(Type=Bone,Name="deform_shoulderpad_r_base_root"),Transform=(Rotation=(X=0.0007,Y=0.0131,Z=-0.0523,W=0.9985),Translation=(X=-2,Y=0,Z=-1.5),Scale3D=(X=1,Y=1,Z=1))))
    WeightMinimum=0
    Weight=@n32.Result
   )
   n33 = RigVMVariableNode() {variable=Skydive Thigh Relative}
   &RigUnit_ModifyTransforms(
    ItemToModify=((Item=(Type=Bone,Name="deform_shoulderpad_r"),Transform=(Rotation=(X=0,Y=0,Z=0,W=1),Translation=(X=1,Y=-0.8,Z=2.2),Scale3D=(X=1,Y=1,Z=1))))
    WeightMinimum=0
    Weight=@n33.Value
    ExecutePin=@RigVMFunction_Sequence_11.C
   )
   > n34 = &RigUnit_ModifyTransforms(
    ItemToModify=((Item=(Type=Bone,Name="deform_shoulderpad_l"),Transform=(Rotation=(X=0,Y=0,Z=0,W=1),Translation=(X=-1,Y=0,Z=-2.2),Scale3D=(X=1,Y=1,Z=1))))
    WeightMinimum=0
    Weight=@n33.Value
   )
   n35 = &RigVMFunction_Sequence(ExecuteContext=@n34.ExecutePin)
   n36 = &RigUnit_GetTransform(Item=(Type=Bone,Name="head"), bInitial=False)
   n37 = &RigUnit_GetTransform(Item=(Type=Bone,Name="deform_shoulderpad_target_up_l"), bInitial=False)
   n38 = &RigVMFunction_AlphaInterp(
    InterpSpeedDecreasing=30
    InterpSpeedIncreasing=30
    Scale=1
    bClampResult=true
    Value=&RigVMFunction_MathDoubleRemap(
     SourceMaximum=15
     SourceMinimum=19
     TargetMaximum=1
     TargetMinimum=0
     Value=&RigVMFunction_MathVectorDistance(B=@n36.Transform.Translation, A=@n37.Transform.Translation).Result
    ).Result
   )
   &RigUnit_ModifyTransforms(
    ItemToModify=((Item=(Type=Bone,Name="deform_shoulderpad_l"),Transform=(Rotation=(X=0,Y=0,Z=0,W=1),Translation=(X=5,Y=0,Z=-3),Scale3D=(X=1,Y=1,Z=1))))
    WeightMinimum=0
    Weight=&RigVMFunction_MathDoubleMul(A=@n38.Result, B=RigVMVariableNode() {variable=Skydive Thigh Relative}.Value).Result
    ExecutePin=@n35.A
   )
   n39 = &RigUnit_GetTransform(Item=(Type=Bone,Name="head"), bInitial=False)
   n40 = &RigUnit_GetTransform(Item=(Type=Bone,Name="deform_shoulderpad_target_up_r"), bInitial=False)
   n41 = &RigVMFunction_AlphaInterp(
    InterpSpeedDecreasing=30
    InterpSpeedIncreasing=30
    Scale=1
    bClampResult=true
    Value=&RigVMFunction_MathDoubleRemap(
     SourceMaximum=14.5
     SourceMinimum=19
     TargetMaximum=1
     TargetMinimum=0
     Value=&RigVMFunction_MathVectorDistance(B=@n39.Transform.Translation, A=@n40.Transform.Translation).Result
    ).Result
   )
   &RigUnit_ModifyTransforms(
    ItemToModify=((Item=(Type=Bone,Name="deform_shoulderpad_r"),Transform=(Rotation=(X=0,Y=0,Z=0,W=1),Translation=(X=-5,Y=0,Z=3),Scale3D=(X=1,Y=1,Z=1))))
    WeightMinimum=0
    Weight=&RigVMFunction_MathDoubleMul(A=@n41.Result, B=RigVMVariableNode() {variable=Skydive Thigh Relative}.Value).Result
    ExecutePin=@n35.B
   )
   Get Delta Ref = RigVMFunctionReferenceNode(Item=(Type=Bone,Name="pinky_01_r"), ExecuteContext=@RigVMFunction_Sequence_1_1.H)
   > &RigUnit_ModifyTransforms(
    ItemToModify=((Item=(Type=Bone,Name="pinky_01_r"),Transform=(Rotation=(X=0,Y=0,Z=-0.0436,W=0.999),Translation=(X=0.1,Y=-1.3,Z=-0.5),Scale3D=(X=1,Y=1,Z=1))))
    WeightMinimum=0
    Weight=&RigVMFunction_MathDoubleRemap(SourceMaximum=34, SourceMinimum=23, TargetMaximum=1, TargetMinimum=0, Value=@"Get Delta Ref".Result.Rotation.Yaw).Result
   )
   RigVMFunction_ControlFlowBranch = &RigVMFunction_ControlFlowBranch(ExecuteContext=@RigVMFunction_Sequence_1_1.I, Condition=RigVMVariableNode() {variable=Lunch Emote}.Value)
   &RigUnit_ModifyTransforms(
    ItemToModify=((Item=(Type=Bone,Name="neck_02"),Transform=(Rotation=(X=0,Y=0,Z=-0.1305,W=0.9914),Translation=(X=2.5,Y=0,Z=0),Scale3D=(X=1,Y=1,Z=1))))
    WeightMinimum=0
    Weight=&RigVMFunction_MathDoubleRemap(
     SourceMaximum=53
     SourceMinimum=68
     TargetMaximum=1
     TargetMinimum=0
     Value=&RigUnit_GetRelativeTransformForItem(Child=(Type=Bone,Name="pelvis"), Parent=(Type=Bone,Name="root"), bChildInitial=False, bParentInitial=False).RelativeTransform.Translation.Z
    ).Result
    ExecutePin=@RigVMFunction_ControlFlowBranch.True
   )
   n42 = &RigVMFunction_Sequence(ExecuteContext=@RigVMFunction_Sequence_1_1.C)
   n43 = &RigUnit_SphericalPoseReader(
    ActiveRegionScaleFactors=(PositiveWidth=1,NegativeWidth=1,PositiveHeight=1,NegativeHeight=1)
    ActiveRegionSize=0.05
    Debug=(bDrawDebug=false,bDraw2D=False,bDrawLocalAxes=False,DebugScale=25,DebugSegments=20,DebugThickness=0.25)
    DriverAxis=(X=1,Y=0,Z=0)
    DriverItem=(Type=Bone,Name="thigh_r")
    FalloffRegionScaleFactors=(PositiveWidth=1,NegativeWidth=1,PositiveHeight=1,NegativeHeight=1)
    FalloffSize=0.45
    FlipHeightScaling=False
    FlipWidthScaling=False
    OptionalParentItem=(Type=Bone,Name="pelvis")
    RotationOffset=(X=10,Y=-90,Z=0)
    ExecutePin=@n42.A
   )
   Add = &RigVMFunction_MathDoubleAdd(B=0.2, A=@n43.OutputParam)
   Clamp = &RigVMFunction_MathDoubleClamp(Value=@Add.Result)
   > &RigUnit_ParentConstraint(
    AdvancedSettings=(InterpolationType=Average,RotationOrderForFilter=ZXY)
    Child=(Type=Bone,Name="deform_thighpad_base_r")
    Filter=(TranslationFilter=(bX=false,bY=false,bZ=false),RotationFilter=(bX=false,bY=true,bZ=false),ScaleFilter=(bX=false,bY=false,bZ=false))
    Parents=((Item=(Type=Bone,Name="pelvis"),Weight=1),(Item=(Type=Bone,Name="thigh_r"),Weight=0.2))
    bMaintainOffset=true
    Weight=@Clamp.Result
   )
   > &RigUnit_ParentConstraint(
    AdvancedSettings=(InterpolationType=Average,RotationOrderForFilter=ZXY)
    Child=(Type=Bone,Name="deform_thighpad_r")
    Filter=(TranslationFilter=(bX=false,bY=false,bZ=true),RotationFilter=(bX=true,bY=true,bZ=false),ScaleFilter=(bX=false,bY=false,bZ=false))
    Parents=((Item=(Type=Bone,Name="pelvis"),Weight=1),(Item=(Type=Bone,Name="deform_thighpad_base_r"),Weight=1))
    bMaintainOffset=true
    Weight=@Clamp.Result
   )
   > &RigUnit_ModifyTransforms(
    ItemToModify=((Item=(Type=Bone,Name="deform_thighpad_base_r"),Transform=(Rotation=(X=-0.0131,Y=-0.0087,Z=-0.0001,W=0.9999),Translation=(X=0,Y=0,Z=0.5),Scale3D=(X=1,Y=1,Z=1))))
    WeightMinimum=0
    Weight=@n43.OutputParam
   )
   > &RigUnit_ModifyTransforms(
    ItemToModify=((Item=(Type=Bone,Name="deform_thighpad_r"),Transform=(Rotation=(X=-0.0087,Y=-0.0175,Z=-0.0002,W=0.9998),Translation=(X=0,Y=0,Z=0),Scale3D=(X=1,Y=1,Z=1))))
    WeightMinimum=0
    Weight=@n43.OutputParam
   )
   > n44 = &RigUnit_SphericalPoseReader(
    ActiveRegionScaleFactors=(PositiveWidth=1,NegativeWidth=1,PositiveHeight=1,NegativeHeight=1)
    ActiveRegionSize=0.01
    Debug=(bDrawDebug=false,bDraw2D=False,bDrawLocalAxes=False,DebugScale=25,DebugSegments=20,DebugThickness=0.25)
    DriverAxis=(X=0,Y=0,Z=1)
    DriverItem=(Type=Bone,Name="thigh_r")
    FalloffRegionScaleFactors=(PositiveWidth=1,NegativeWidth=1,PositiveHeight=1,NegativeHeight=1)
    FalloffSize=0.53
    FlipHeightScaling=False
    FlipWidthScaling=False
    OptionalParentItem=(Type=Bone,Name="pelvis")
    RotationOffset=(X=90,Y=0,Z=90)
   )
   > n45 = &RigUnit_SphericalPoseReader(
    ActiveRegionScaleFactors=(PositiveWidth=1,NegativeWidth=0,PositiveHeight=1,NegativeHeight=1)
    ActiveRegionSize=0.01
    Debug=(bDrawDebug=false,bDraw2D=False,bDrawLocalAxes=False,DebugScale=25,DebugSegments=20,DebugThickness=0.25)
    DriverAxis=(X=1,Y=0,Z=0)
    DriverItem=(Type=Bone,Name="thigh_r")
    FalloffRegionScaleFactors=(PositiveWidth=1,NegativeWidth=1,PositiveHeight=1,NegativeHeight=1)
    FalloffSize=0.53
    FlipHeightScaling=False
    FlipWidthScaling=False
    OptionalParentItem=(Type=Bone,Name="pelvis")
    RotationOffset=(X=0,Y=0,Z=0)
   )
   Multiply_1 = RigVMAggregateNode(C=1, B=@n45.OutputParam, A=@n44.OutputParam)
   > &RigUnit_ModifyTransforms(
    ItemToModify=((Item=(Type=Bone,Name="deform_thighpad_base_r"),Transform=(Rotation=(X=0.8192,Y=0,Z=0.5736,W=0),Translation=(X=0,Y=0,Z=0),Scale3D=(X=1,Y=1,Z=1))))
    Mode=OverrideLocal
    WeightMinimum=0
    Weight=@Multiply_1.Result
   )
   > &RigUnit_ModifyTransforms(
    ItemToModify=((Item=(Type=Bone,Name="thigh_r"),Transform=(Rotation=(X=0,Y=0,Z=0,W=1),Translation=(X=0,Y=0,Z=0),Scale3D=(X=1,Y=1,Z=1))))
    Weight=1
    WeightMinimum=0
   )
   > n46 = &RigUnit_SphericalPoseReader(
    ActiveRegionScaleFactors=(PositiveWidth=1,NegativeWidth=1,PositiveHeight=1,NegativeHeight=1)
    ActiveRegionSize=0.05
    Debug=(bDrawDebug=false,bDraw2D=False,bDrawLocalAxes=False,DebugScale=25,DebugSegments=20,DebugThickness=0.25)
    DriverAxis=(X=1,Y=0,Z=0)
    DriverItem=(Type=Bone,Name="thigh_r")
    FalloffRegionScaleFactors=(PositiveWidth=1,NegativeWidth=1,PositiveHeight=1,NegativeHeight=1)
    FalloffSize=0.2
    FlipHeightScaling=False
    FlipWidthScaling=False
    OptionalParentItem=(Type=Bone,Name="pelvis")
    RotationOffset=(X=-120,Y=20,Z=0)
   )
   n47 = &RigVMFunction_AlphaInterp(InterpSpeedDecreasing=30, InterpSpeedIncreasing=30, Scale=1, bClampResult=true, Value=@n46.OutputParam)
   > &RigUnit_ModifyTransforms(
    ItemToModify=((Item=(Type=Bone,Name="deform_thighpad_base_r"),Transform=(Rotation=(X=0,Y=0.0785,Z=0,W=0.9969),Translation=(X=0,Y=0,Z=0),Scale3D=(X=1,Y=1,Z=1))))
    WeightMinimum=0
    Weight=@n47.Result
   )
   > &RigUnit_ModifyTransforms(
    ItemToModify=((Item=(Type=Bone,Name="deform_thighpad_r"),Transform=(Rotation=(X=0,Y=0.0436,Z=0,W=0.999),Translation=(X=0,Y=0,Z=0),Scale3D=(X=1,Y=1,Z=1))))
    WeightMinimum=0
    Weight=@n47.Result
   )
   > n48 = &RigUnit_SphericalPoseReader(
    ActiveRegionScaleFactors=(PositiveWidth=1,NegativeWidth=1,PositiveHeight=1,NegativeHeight=1)
    ActiveRegionSize=0.05
    Debug=(bDrawDebug=false,bDraw2D=False,bDrawLocalAxes=False,DebugScale=25,DebugSegments=20,DebugThickness=0.25)
    DriverAxis=(X=1,Y=0,Z=0)
    DriverItem=(Type=Bone,Name="calf_r")
    FalloffRegionScaleFactors=(PositiveWidth=1,NegativeWidth=1,PositiveHeight=1,NegativeHeight=1)
    FalloffSize=0.2
    FlipHeightScaling=False
    FlipWidthScaling=False
    OptionalParentItem=(Type=Bone,Name="spine_05")
    RotationOffset=(X=-2.8,Y=194.83,Z=0.15)
   )
   > &RigUnit_ModifyTransforms(
    ItemToModify=((Item=(Type=Bone,Name="deform_thighpad_base_r"),Transform=(Rotation=(X=-0.1304,Y=-0.0432,Z=-0.0057,W=0.9905),Translation=(X=0,Y=0,Z=0),Scale3D=(X=1,Y=1,Z=1))))
    WeightMinimum=0
    Weight=@n48.OutputParam
   )
   > &RigUnit_ModifyTransforms(
    ItemToModify=((Item=(Type=Bone,Name="deform_thighpad_r"),Transform=(Rotation=(X=-0.0175,Y=-0.0087,Z=-0.0002,W=0.9998),Translation=(X=0,Y=0,Z=0.8),Scale3D=(X=1,Y=1,Z=1))))
    WeightMinimum=0
    Weight=@n48.OutputParam
   )
   n49 = &RigVMFunction_MathDoubleRemap(
    SourceMaximum=8
    SourceMinimum=11
    TargetMaximum=1
    TargetMinimum=0
    Value=&RigUnit_GetRelativeTransformForItem(Child=(Type=Bone,Name="thigh_r"), Parent=(Type=Bone,Name="pelvis"), bChildInitial=False, bParentInitial=False).RelativeTransform.Translation.Z
   )
   > RigVMVariableNode(Value=@n49.Result) {variable=Skydive Thigh Relative}
   > &RigUnit_ModifyTransforms(
    ItemToModify=((Item=(Type=Bone,Name="thigh_r"),Transform=(Rotation=(X=0,Y=0,Z=0,W=1),Translation=(X=0,Y=0,Z=2),Scale3D=(X=1,Y=1,Z=1))))
    WeightMinimum=0
    Weight=@n49.Result
   )
   > &RigUnit_ModifyTransforms(
    ItemToModify=((Item=(Type=Bone,Name="deform_thighpad_base_r"),Transform=(Rotation=(X=0,Y=-0.0436,Z=0,W=0.999),Translation=(X=0,Y=0,Z=1),Scale3D=(X=1,Y=1,Z=1))))
    WeightMinimum=0
    Weight=@n49.Result
   )
   > &RigUnit_ModifyTransforms(
    ItemToModify=((Item=(Type=Bone,Name="deform_thighpad_r"),Transform=(Rotation=(X=0.0268,Y=0.0255,Z=-0.0268,W=0.999),Translation=(X=0,Y=0,Z=-2.5),Scale3D=(X=0.85,Y=1,Z=0.96))))
    WeightMinimum=0
    Weight=@n49.Result
   )
   n50 = &RigUnit_SphericalPoseReader(
    ActiveRegionScaleFactors=(PositiveWidth=1,NegativeWidth=1,PositiveHeight=1,NegativeHeight=1)
    ActiveRegionSize=0.05
    Debug=(bDrawDebug=false,bDraw2D=False,bDrawLocalAxes=False,DebugScale=25,DebugSegments=20,DebugThickness=0.25)
    DriverAxis=(X=-1,Y=0,Z=0)
    DriverItem=(Type=Bone,Name="thigh_l")
    FalloffRegionScaleFactors=(PositiveWidth=1,NegativeWidth=1,PositiveHeight=1,NegativeHeight=1)
    FalloffSize=0.45
    FlipHeightScaling=False
    FlipWidthScaling=False
    OptionalParentItem=(Type=Bone,Name="pelvis")
    RotationOffset=(X=-10,Y=90,Z=0)
    ExecutePin=@n42.B
   )
   n51 = &RigVMFunction_MathDoubleClamp(Value=&RigVMFunction_MathDoubleAdd(B=0.2, A=@n50.OutputParam).Result)
   > &RigUnit_ParentConstraint(
    AdvancedSettings=(InterpolationType=Average,RotationOrderForFilter=ZXY)
    Child=(Type=Bone,Name="deform_thighpad_base_l")
    Filter=(TranslationFilter=(bX=false,bY=false,bZ=false),RotationFilter=(bX=false,bY=true,bZ=false),ScaleFilter=(bX=false,bY=false,bZ=false))
    Parents=((Item=(Type=Bone,Name="pelvis"),Weight=1),(Item=(Type=Bone,Name="thigh_l"),Weight=0.2))
    bMaintainOffset=true
    Weight=@n51.Result
   )
   > &RigUnit_ParentConstraint(
    AdvancedSettings=(InterpolationType=Average,RotationOrderForFilter=ZXY)
    Child=(Type=Bone,Name="deform_thighpad_l")
    Filter=(TranslationFilter=(bX=false,bY=false,bZ=true),RotationFilter=(bX=true,bY=true,bZ=false),ScaleFilter=(bX=false,bY=false,bZ=false))
    Parents=((Item=(Type=Bone,Name="pelvis"),Weight=1),(Item=(Type=Bone,Name="deform_thighpad_base_l"),Weight=1))
    bMaintainOffset=true
    Weight=@n51.Result
   )
   > &RigUnit_ModifyTransforms(
    ItemToModify=((Item=(Type=Bone,Name="deform_thighpad_base_l"),Transform=(Rotation=(X=-0.0131,Y=-0.0087,Z=-0.0001,W=0.9999),Translation=(X=0,Y=0,Z=-0.5),Scale3D=(X=1,Y=1,Z=1))))
    WeightMinimum=0
    Weight=@n50.OutputParam
   )
   > &RigUnit_ModifyTransforms(
    ItemToModify=((Item=(Type=Bone,Name="deform_thighpad_l"),Transform=(Rotation=(X=-0.0087,Y=-0.0175,Z=-0.0002,W=0.9998),Translation=(X=0,Y=0,Z=0),Scale3D=(X=1,Y=1,Z=1))))
    WeightMinimum=0
    Weight=@n50.OutputParam
   )
   > n52 = &RigUnit_SphericalPoseReader(
    ActiveRegionScaleFactors=(PositiveWidth=1,NegativeWidth=1,PositiveHeight=1,NegativeHeight=1)
    ActiveRegionSize=0.01
    Debug=(bDrawDebug=false,bDraw2D=False,bDrawLocalAxes=False,DebugScale=25,DebugSegments=20,DebugThickness=0.25)
    DriverAxis=(X=0,Y=0,Z=-1)
    DriverItem=(Type=Bone,Name="thigh_l")
    FalloffRegionScaleFactors=(PositiveWidth=1,NegativeWidth=1,PositiveHeight=1,NegativeHeight=1)
    FalloffSize=0.53
    FlipHeightScaling=False
    FlipWidthScaling=False
    OptionalParentItem=(Type=Bone,Name="pelvis")
    RotationOffset=(X=-90,Y=0,Z=90)
   )
   > n53 = &RigUnit_SphericalPoseReader(
    ActiveRegionScaleFactors=(PositiveWidth=1,NegativeWidth=0,PositiveHeight=1,NegativeHeight=1)
    ActiveRegionSize=0.01
    Debug=(bDrawDebug=false,bDraw2D=False,bDrawLocalAxes=False,DebugScale=25,DebugSegments=20,DebugThickness=0.25)
    DriverAxis=(X=-1,Y=0,Z=0)
    DriverItem=(Type=Bone,Name="thigh_l")
    FalloffRegionScaleFactors=(PositiveWidth=1,NegativeWidth=1,PositiveHeight=1,NegativeHeight=1)
    FalloffSize=0.53
    FlipHeightScaling=False
    FlipWidthScaling=False
    OptionalParentItem=(Type=Bone,Name="pelvis")
    RotationOffset=(X=180,Y=0,Z=0)
   )
   Multiply_1_1 = RigVMAggregateNode(C=1, A=@n52.OutputParam, B=@n53.OutputParam)
   > &RigUnit_ModifyTransforms(
    ItemToModify=((Item=(Type=Bone,Name="deform_thighpad_base_l"),Transform=(Rotation=(X=0.8192,Y=0,Z=0.5736,W=0),Translation=(X=0,Y=0,Z=0),Scale3D=(X=1,Y=1,Z=1))))
    Mode=OverrideLocal
    WeightMinimum=0
    Weight=@Multiply_1_1.Result
   )
   > &RigUnit_ModifyTransforms(
    ItemToModify=((Item=(Type=Bone,Name="thigh_l"),Transform=(Rotation=(X=0,Y=0,Z=0,W=1),Translation=(X=0,Y=0,Z=0),Scale3D=(X=1,Y=1,Z=1))))
    Weight=1
    WeightMinimum=0
   )
   > n54 = &RigUnit_SphericalPoseReader(
    ActiveRegionScaleFactors=(PositiveWidth=1,NegativeWidth=1,PositiveHeight=1,NegativeHeight=1)
    ActiveRegionSize=0.05
    Debug=(bDrawDebug=false,bDraw2D=False,bDrawLocalAxes=False,DebugScale=25,DebugSegments=20,DebugThickness=0.25)
    DriverAxis=(X=-1,Y=0,Z=0)
    DriverItem=(Type=Bone,Name="thigh_l")
    FalloffRegionScaleFactors=(PositiveWidth=1,NegativeWidth=1,PositiveHeight=1,NegativeHeight=1)
    FalloffSize=0.2
    FlipHeightScaling=False
    FlipWidthScaling=False
    OptionalParentItem=(Type=Bone,Name="pelvis")
    RotationOffset=(X=60,Y=20,Z=0)
   )
   AlphaInterp = &RigVMFunction_AlphaInterp(InterpSpeedDecreasing=30, InterpSpeedIncreasing=30, Scale=1, bClampResult=true, Value=@n54.OutputParam)
   > &RigUnit_ModifyTransforms(
    ItemToModify=((Item=(Type=Bone,Name="deform_thighpad_base_l"),Transform=(Rotation=(X=0,Y=0.0785,Z=0,W=0.9969),Translation=(X=0,Y=0,Z=0),Scale3D=(X=1,Y=1,Z=1))))
    WeightMinimum=0
    Weight=@AlphaInterp.Result
   )
   > &RigUnit_ModifyTransforms(
    ItemToModify=((Item=(Type=Bone,Name="deform_thighpad_l"),Transform=(Rotation=(X=0,Y=0.0436,Z=0,W=0.999),Translation=(X=0,Y=0,Z=0),Scale3D=(X=1,Y=1,Z=1))))
    WeightMinimum=0
    Weight=@AlphaInterp.Result
   )
   > n55 = &RigUnit_SphericalPoseReader(
    ActiveRegionScaleFactors=(PositiveWidth=1,NegativeWidth=1,PositiveHeight=1,NegativeHeight=1)
    ActiveRegionSize=0.05
    Debug=(bDrawDebug=false,bDraw2D=False,bDrawLocalAxes=False,DebugScale=25,DebugSegments=20,DebugThickness=0.25)
    DriverAxis=(X=-1,Y=0,Z=0)
    DriverItem=(Type=Bone,Name="calf_l")
    FalloffRegionScaleFactors=(PositiveWidth=1,NegativeWidth=1,PositiveHeight=1,NegativeHeight=1)
    FalloffSize=0.2
    FlipHeightScaling=False
    FlipWidthScaling=False
    OptionalParentItem=(Type=Bone,Name="spine_05")
    RotationOffset=(X=2.8,Y=14.83,Z=-0.15)
   )
   > &RigUnit_ModifyTransforms(
    ItemToModify=((Item=(Type=Bone,Name="deform_thighpad_base_l"),Transform=(Rotation=(X=-0.1131,Y=-0.0347,Z=-0.004,W=0.993),Translation=(X=0,Y=0,Z=0),Scale3D=(X=1,Y=1,Z=1))))
    WeightMinimum=0
    Weight=@n55.OutputParam
   )
   > &RigUnit_ModifyTransforms(
    ItemToModify=((Item=(Type=Bone,Name="deform_thighpad_l"),Transform=(Rotation=(X=0,Y=-0.0087,Z=0,W=1),Translation=(X=0,Y=0,Z=-0.8),Scale3D=(X=1,Y=1,Z=1))))
    WeightMinimum=0
    Weight=@n55.OutputParam
   )
   Remap = &RigVMFunction_MathDoubleRemap(
    SourceMaximum=-8
    SourceMinimum=-11
    TargetMaximum=1
    TargetMinimum=0
    Value=&RigUnit_GetRelativeTransformForItem(Child=(Type=Bone,Name="thigh_l"), Parent=(Type=Bone,Name="pelvis"), bChildInitial=False, bParentInitial=False).RelativeTransform.Translation.Z
   )
   > &RigUnit_ModifyTransforms(
    ItemToModify=((Item=(Type=Bone,Name="thigh_l"),Transform=(Rotation=(X=0,Y=0,Z=0,W=1),Translation=(X=0,Y=0,Z=-2),Scale3D=(X=1,Y=1,Z=1))))
    WeightMinimum=0
    Weight=@Remap.Result
   )
   > &RigUnit_ModifyTransforms(
    ItemToModify=((Item=(Type=Bone,Name="deform_thighpad_base_l"),Transform=(Rotation=(X=0,Y=-0.0436,Z=0,W=0.999),Translation=(X=0,Y=0,Z=-1),Scale3D=(X=1,Y=1,Z=1))))
    WeightMinimum=0
    Weight=@Remap.Result
   )
   > &RigUnit_ModifyTransforms(
    ItemToModify=((Item=(Type=Bone,Name="deform_thighpad_l"),Transform=(Rotation=(X=0.0268,Y=0.0255,Z=-0.0268,W=0.999),Translation=(X=0,Y=0,Z=2.5),Scale3D=(X=0.85,Y=1,Z=0.96))))
    WeightMinimum=0
    Weight=@Remap.Result
   )
   n56 = &RigUnit_SphericalPoseReader(
    ActiveRegionScaleFactors=(PositiveWidth=1,NegativeWidth=1,PositiveHeight=1,NegativeHeight=1)
    ActiveRegionSize=0.12
    Debug=(bDrawDebug=false,bDraw2D=False,bDrawLocalAxes=False,DebugScale=25,DebugSegments=20,DebugThickness=0.25)
    DriverAxis=(X=0,Y=1,Z=0)
    DriverItem=(Type=Bone,Name="spine_03")
    FalloffRegionScaleFactors=(PositiveWidth=1,NegativeWidth=1,PositiveHeight=1,NegativeHeight=1)
    FalloffSize=0.2
    FlipHeightScaling=False
    FlipWidthScaling=False
    OptionalParentItem=(Type=Bone,Name="root")
    RotationOffset=(X=0,Y=90,Z=0)
    ExecutePin=@RigVMFunction_Sequence_1_1.B
   )
   > n57 = &RigUnit_SphericalPoseReader(
    ActiveRegionScaleFactors=(PositiveWidth=1,NegativeWidth=1,PositiveHeight=1,NegativeHeight=1)
    ActiveRegionSize=0.12
    Debug=(bDrawDebug=false,bDraw2D=False,bDrawLocalAxes=False,DebugScale=25,DebugSegments=20,DebugThickness=0.25)
    DriverAxis=(X=0,Y=1,Z=0)
    DriverItem=(Type=Bone,Name="pelvis")
    FalloffRegionScaleFactors=(PositiveWidth=1,NegativeWidth=1,PositiveHeight=1,NegativeHeight=1)
    FalloffSize=0.2
    FlipHeightScaling=False
    FlipWidthScaling=False
    OptionalParentItem=(Type=Bone,Name="root")
    RotationOffset=(X=20,Y=90,Z=0)
   )
   Multiply = &RigVMFunction_MathDoubleMul(A=@n57.OutputParam, B=@n56.OutputParam)
   > RigVMVariableNode(Value=&RigVMFunction_MathDoubleRemap(SourceMaximum=1, SourceMinimum=0, TargetMaximum=0, TargetMinimum=1, Value=@Multiply.Result).Result) {variable=DBNO disabled}
   n58 = &RigUnit_SphericalPoseReader(
    ActiveRegionScaleFactors=(PositiveWidth=1,NegativeWidth=1,PositiveHeight=1,NegativeHeight=1)
    ActiveRegionSize=0.07
    Debug=(bDrawDebug=false,bDraw2D=False,bDrawLocalAxes=False,DebugScale=25,DebugSegments=20,DebugThickness=0.25)
    DriverAxis=(X=1,Y=0,Z=0)
    DriverItem=(Type=Bone,Name="head")
    FalloffRegionScaleFactors=(PositiveWidth=1,NegativeWidth=1,PositiveHeight=1,NegativeHeight=1)
    FalloffSize=0.15
    FlipHeightScaling=False
    FlipWidthScaling=False
    OptionalParentItem=(Type=Bone,Name="spine_05")
    RotationOffset=(X=90,Y=0,Z=-142)
    ExecutePin=@RigVMFunction_Sequence_1_1.F
   )
   n59 = &RigVMFunction_MathDoubleMul(A=@n58.OutputParam, B=RigVMVariableNode() {variable=Skydive Thigh Relative}.Value)
   > &RigUnit_ModifyTransforms(
    ItemToModify=((Item=(Type=Bone,Name="neck_01"),Transform=(Rotation=(X=0,Y=0,Z=0.0436,W=0.999),Translation=(X=0,Y=6,Z=0),Scale3D=(X=1,Y=1,Z=1))))
    WeightMinimum=0
    Weight=@n59.Result
   )
   > &RigUnit_ModifyTransforms(
    ItemToModify=((Item=(Type=Bone,Name="neck_02"),Transform=(Rotation=(X=0,Y=0,Z=0.0436,W=0.999),Translation=(X=0,Y=6,Z=0),Scale3D=(X=1,Y=1,Z=1))))
    WeightMinimum=0
    Weight=@n59.Result
   )
   RigVMFunction_Sequence = &RigVMFunction_Sequence(ExecuteContext=@RigVMFunction_Sequence_1_1.D)
   n60 = &RigUnit_SphericalPoseReader(
    ActiveRegionScaleFactors=(PositiveWidth=1,NegativeWidth=1,PositiveHeight=1,NegativeHeight=1)
    ActiveRegionSize=0.05
    Debug=(bDrawDebug=false,bDraw2D=false,bDrawLocalAxes=false,DebugScale=25,DebugSegments=20,DebugThickness=0.25)
    DriverAxis=(X=-1,Y=0,Z=0)
    DriverItem=(Type=Bone,Name="thigh_r")
    FalloffRegionScaleFactors=(PositiveWidth=1,NegativeWidth=1,PositiveHeight=1,NegativeHeight=1)
    FalloffSize=0.2
    FlipHeightScaling=False
    FlipWidthScaling=False
    OptionalParentItem=(Type=Bone,Name="pelvis")
    RotationOffset=(X=-4,Y=96,Z=0)
    ExecutePin=@RigVMFunction_Sequence.B
   )
   n61 = &RigVMFunction_AlphaInterp(InterpSpeedDecreasing=30, InterpSpeedIncreasing=30, Scale=1, bClampResult=False, Value=@n60.OutputParam)
   > &RigUnit_ModifyTransforms(
    ItemToModify=((Item=(Type=Bone,Name="deform_buckler_b"),Transform=(Rotation=(X=0,Y=0,Z=0,W=1),Translation=(X=-0.6,Y=-1,Z=0.5),Scale3D=(X=0.95,Y=1,Z=1))))
    WeightMinimum=0
    Weight=@n61.Result
   )
   > n62 = &RigUnit_SphericalPoseReader(
    ActiveRegionScaleFactors=(PositiveWidth=1,NegativeWidth=1,PositiveHeight=1,NegativeHeight=1)
    ActiveRegionSize=0.05
    Debug=(bDrawDebug=false,bDraw2D=false,bDrawLocalAxes=false,DebugScale=25,DebugSegments=20,DebugThickness=0.25)
    DriverAxis=(X=1,Y=0,Z=0)
    DriverItem=(Type=Bone,Name="thigh_l")
    FalloffRegionScaleFactors=(PositiveWidth=1,NegativeWidth=1,PositiveHeight=1,NegativeHeight=1)
    FalloffSize=0.2
    FlipHeightScaling=False
    FlipWidthScaling=False
    OptionalParentItem=(Type=Bone,Name="pelvis")
    RotationOffset=(X=4,Y=-84,Z=0)
   )
   n63 = &RigVMFunction_AlphaInterp(InterpSpeedDecreasing=30, InterpSpeedIncreasing=30, Scale=1, bClampResult=False, Value=@n62.OutputParam)
   > &RigUnit_ModifyTransforms(
    ItemToModify=((Item=(Type=Bone,Name="deform_buckler_b"),Transform=(Rotation=(X=0,Y=0,Z=0,W=1),Translation=(X=0.6,Y=-1,Z=0.5),Scale3D=(X=0.95,Y=1,Z=1))))
    WeightMinimum=0
    Weight=@n63.Result
   )
   &RigUnit_ParentConstraint(
    AdvancedSettings=(InterpolationType=Average,RotationOrderForFilter=XZY)
    Child=(Type=Bone,Name="deform_buckler_b")
    Filter=(TranslationFilter=(bX=false,bY=false,bZ=false),RotationFilter=(bX=true,bY=true,bZ=false),ScaleFilter=(bX=True,bY=True,bZ=True))
    Parents=((Item=(Type=Bone,Name="pelvis"),Weight=1),(Item=(Type=Bone,Name="spine_01"),Weight=0.8),(Item=(Type=Bone,Name="spine_02"),Weight=0.6))
    Weight=1
    bMaintainOffset=true
    ExecutePin=@RigVMFunction_Sequence.A
   )
   > &RigUnit_ParentConstraint(
    AdvancedSettings=(InterpolationType=Average,RotationOrderForFilter=XZY)
    Child=(Type=Bone,Name="deform_buckler_b")
    Filter=(TranslationFilter=(bX=true,bY=false,bZ=true),RotationFilter=(bX=false,bY=false,bZ=false),ScaleFilter=(bX=True,bY=True,bZ=True))
    Parents=((Item=(Type=Bone,Name="pelvis"),Weight=1),(Item=(Type=Bone,Name="spine_01"),Weight=0.8),(Item=(Type=Bone,Name="spine_02"),Weight=0.6))
    Weight=0.5
    bMaintainOffset=true
   )
   > &RigUnit_ParentConstraint(
    AdvancedSettings=(InterpolationType=Average,RotationOrderForFilter=XZY)
    Child=(Type=Bone,Name="deform_buckler_b")
    Filter=(TranslationFilter=(bX=false,bY=false,bZ=false),RotationFilter=(bX=false,bY=false,bZ=true),ScaleFilter=(bX=True,bY=True,bZ=True))
    Parents=((Item=(Type=Bone,Name="spine_03"),Weight=1))
    Weight=0.6
   )
   RigVMFunction_MathQuaternionToEuler = &RigVMFunction_MathQuaternionToEuler(
    RotationOrder=XZY
    Value=&RigUnit_GetRelativeTransformForItem(Child=(Type=Bone,Name="thigh_r"), Parent=(Type=Bone,Name="pelvis"), bChildInitial=False, bParentInitial=False).RelativeTransform.Rotation
   )
   n64 = &RigVMFunction_MathQuaternionToEuler(
    RotationOrder=XZY
    Value=&RigUnit_GetRelativeTransformForItem(Child=(Type=Bone,Name="thigh_l"), Parent=(Type=Bone,Name="pelvis"), bChildInitial=False, bParentInitial=False).RelativeTransform.Rotation
   )
   > &RigUnit_ModifyTransforms(
    ItemToModify=((Item=(Type=Bone,Name="deform_buckler_b"),Transform=(Rotation=(X=-0.0087,Y=0,Z=0,W=1),Translation=(X=0,Y=-4,Z=1),Scale3D=(X=1,Y=1,Z=1))))
    WeightMinimum=0
    Weight=&RigVMFunction_MathDoubleMax(
     A=&RigVMFunction_MathDoubleRemap(SourceMaximum=85, SourceMinimum=-2, TargetMaximum=1, TargetMinimum=0, Value=@RigVMFunction_MathQuaternionToEuler.Result.Z).Result
     B=&RigVMFunction_MathDoubleRemap(SourceMaximum=-85, SourceMinimum=2, TargetMaximum=1, TargetMinimum=0, Value=@n64.Result.Z).Result
    ).Result
   )
   > n65 = &RigUnit_SphericalPoseReader(
    ActiveRegionScaleFactors=(PositiveWidth=1,NegativeWidth=1,PositiveHeight=1,NegativeHeight=1)
    ActiveRegionSize=0.08
    Debug=(bDrawDebug=false,bDraw2D=false,bDrawLocalAxes=false,DebugScale=25,DebugSegments=20,DebugThickness=0.25)
    DriverAxis=(X=0,Y=1,Z=0)
    DriverItem=(Type=Bone,Name="pelvis")
    FalloffRegionScaleFactors=(PositiveWidth=1,NegativeWidth=1,PositiveHeight=1,NegativeHeight=1)
    FalloffSize=0.2
    FlipHeightScaling=False
    FlipWidthScaling=False
    OptionalParentItem=(Type=Bone,Name="root")
    RotationOffset=(X=0,Y=90,Z=-30)
   )
   > n66 = &RigUnit_SphericalPoseReader(
    ActiveRegionScaleFactors=(PositiveWidth=1,NegativeWidth=1,PositiveHeight=1,NegativeHeight=1)
    ActiveRegionSize=0.1
    Debug=(bDrawDebug=false,bDraw2D=false,bDrawLocalAxes=false,DebugScale=25,DebugSegments=20,DebugThickness=0.25)
    DriverAxis=(X=-1,Y=0,Z=0)
    DriverItem=(Type=Bone,Name="thigh_l")
    FalloffRegionScaleFactors=(PositiveWidth=1,NegativeWidth=1,PositiveHeight=1,NegativeHeight=1)
    FalloffSize=0.2
    FlipHeightScaling=False
    FlipWidthScaling=False
    OptionalParentItem=(Type=Bone,Name="root")
    RotationOffset=(X=-10,Y=95,Z=-15)
   )
   > n67 = &RigUnit_SphericalPoseReader(
    ActiveRegionScaleFactors=(PositiveWidth=1,NegativeWidth=1,PositiveHeight=1,NegativeHeight=1)
    ActiveRegionSize=0.1
    Debug=(bDrawDebug=false,bDraw2D=false,bDrawLocalAxes=false,DebugScale=25,DebugSegments=20,DebugThickness=0.25)
    DriverAxis=(X=1,Y=0,Z=0)
    DriverItem=(Type=Bone,Name="thigh_r")
    FalloffRegionScaleFactors=(PositiveWidth=1,NegativeWidth=1,PositiveHeight=1,NegativeHeight=1)
    FalloffSize=0.2
    FlipHeightScaling=False
    FlipWidthScaling=False
    OptionalParentItem=(Type=Bone,Name="root")
    RotationOffset=(X=10,Y=-85,Z=-15)
   )
   n68 = &RigVMFunction_MathDoubleAdd(B=@n66.OutputParam, A=@n67.OutputParam)
   n69 = &RigVMFunction_AlphaInterp(
    InterpSpeedDecreasing=20
    InterpSpeedIncreasing=20
    Scale=1
    bClampResult=False
    Value=&RigVMFunction_MathDoubleMul(B=&RigVMFunction_MathDoubleClamp(Value=@n68.Result).Result, A=@n65.OutputParam).Result
   )
   > &RigUnit_ModifyTransforms(
    ItemToModify=((Item=(Type=Bone,Name="deform_buckler_b"),Transform=(Rotation=(X=0.0697,Y=-0.0012,Z=-0.0174,W=0.9974),Translation=(X=0,Y=1,Z=-1),Scale3D=(X=1,Y=1,Z=1.1))))
    WeightMinimum=0
    Weight=@n69.Result
   )
   > &RigUnit_ModifyTransforms(
    ItemToModify=((Item=(Type=Bone,Name="deform_buckler_b"),Transform=(Rotation=(X=0.0436,Y=0,Z=0,W=0.999),Translation=(X=0,Y=-0.5,Z=-3),Scale3D=(X=0.85,Y=0.9,Z=0.9))))
    WeightMinimum=0
    Weight=RigVMVariableNode() {variable=Skydive Thigh Relative}.Value
   )
   n70 = &RigVMFunction_Sequence(ExecuteContext=@RigVMFunction_Sequence_1_1.G)
   &RigUnit_ModifyTransforms(
    ItemToModify=((Item=(Type=Bone,Name="foot_l"),Transform=(Rotation=(X=0,Y=0,Z=0,W=1),Translation=(X=0,Y=0,Z=0),Scale3D=(X=1,Y=1,Z=1))))
    Weight=1
    WeightMinimum=0
    ExecutePin=@n70.A
   )
   > Get Delta Ref_1 = RigVMFunctionReferenceNode(Item=(Type=Bone,Name="foot_l"))
   n71 = &RigVMFunction_MathDoubleAbs(Value=@"Get Delta Ref_1".Result.Rotation.Roll)
   n72 = &RigVMFunction_MathDoubleAbs(Value=@"Get Delta Ref_1".Result.Rotation.Pitch)
   > &RigUnit_ModifyTransforms(
    ItemToModify=((Item=(Type=Bone,Name="deform_ankle_l"),Transform=(Rotation=(X=0,Y=0,Z=0,W=1),Translation=(X=0,Y=0,Z=1.6),Scale3D=(X=1,Y=1,Z=1))))
    WeightMinimum=0
    Weight=&RigVMFunction_MathDoubleMax(
     A=&RigVMFunction_MathDoubleRemap(SourceMaximum=20, SourceMinimum=0, TargetMaximum=1, TargetMinimum=0, Value=@n71.Result).Result
     B=&RigVMFunction_MathDoubleRemap(SourceMaximum=20, SourceMinimum=0, TargetMaximum=1, TargetMinimum=0, Value=@n72.Result).Result
    ).Result
   )
   > &RigUnit_ModifyTransforms(
    ItemToModify=((Item=(Type=Bone,Name="root"),Transform=(Rotation=(X=0,Y=0,Z=0,W=1),Translation=(X=0,Y=0,Z=0),Scale3D=(X=1,Y=1,Z=1))))
    Weight=1
    WeightMinimum=0
   )
   n73 = &RigUnit_GetTransform(Item=(Type=Bone,Name="root"), Space=LocalSpace, bInitial=false)
   Divide = &RigVMFunction_MathDoubleDiv(B=2, A=@n73.Transform.Translation.Z)
   RigVMFunction_MathBoolToFloat = &RigVMFunction_MathBoolToFloat(Value=RigVMVariableNode() {variable=isDBNOCarried}.Value)
   Subtract = &RigVMFunction_MathDoubleSub(A=1, B=@RigVMFunction_MathBoolToFloat.Result)
   > &RigUnit_ModifyTransforms(
    ItemToModify=((Item=(Type=Bone,Name="foot_r"),Transform=(Rotation=(X=0,Y=0,Z=0,W=1),Translation=(X=0,Y=0,Z=0),Scale3D=(X=1,Y=1,Z=1))))
    WeightMinimum=0
    ItemToModify.0.Transform.Translation.X=@Divide.Result
    Weight=@Subtract.Result
   )
   > &RigUnit_ModifyTransforms(
    ItemToModify=((Item=(Type=Bone,Name="foot_l"),Transform=(Rotation=(X=0,Y=0,Z=0,W=1),Translation=(X=0,Y=0,Z=0),Scale3D=(X=1,Y=1,Z=1))))
    WeightMinimum=0
    ItemToModify.0.Transform.Translation.X=&RigVMFunction_MathDoubleDiv(B=-2, A=@n73.Transform.Translation.Z).Result
    Weight=@Subtract.Result
   )
   n74 = RigVMVariableNode() {variable=Skydive Thigh Relative}
   > &RigUnit_ModifyTransforms(
    ItemToModify=((Item=(Type=Bone,Name="foot_r"),Transform=(Rotation=(X=0,Y=0,Z=0,W=1),Translation=(X=-1.7,Y=1,Z=0),Scale3D=(X=1,Y=1,Z=1))))
    WeightMinimum=0
    Weight=@n74.Value
   )
   > &RigUnit_ModifyTransforms(
    ItemToModify=((Item=(Type=Bone,Name="foot_l"),Transform=(Rotation=(X=0,Y=0,Z=0,W=1),Translation=(X=1.7,Y=-1,Z=0),Scale3D=(X=1,Y=1,Z=1))))
    WeightMinimum=0
    Weight=@n74.Value
   )
   &RigUnit_ModifyTransforms(
    ItemToModify=((Item=(Type=Bone,Name="foot_r"),Transform=(Rotation=(X=0,Y=0,Z=0,W=1),Translation=(X=0,Y=0,Z=0),Scale3D=(X=1,Y=1,Z=1))))
    Weight=1
    WeightMinimum=0
    ExecutePin=@n70.B
   )
   > Get Delta Ref_1_1 = RigVMFunctionReferenceNode(Item=(Type=Bone,Name="foot_r"))
   n75 = &RigVMFunction_MathDoubleAbs(Value=@"Get Delta Ref_1_1".Result.Rotation.Roll)
   n76 = &RigVMFunction_MathDoubleAbs(Value=@"Get Delta Ref_1_1".Result.Rotation.Pitch)
   > &RigUnit_ModifyTransforms(
    ItemToModify=((Item=(Type=Bone,Name="deform_ankle_r"),Transform=(Rotation=(X=0,Y=0,Z=0,W=1),Translation=(X=0,Y=0,Z=-1.6),Scale3D=(X=1,Y=1,Z=1))))
    WeightMinimum=0
    Weight=&RigVMFunction_MathDoubleMax(
     A=&RigVMFunction_MathDoubleRemap(SourceMaximum=20, SourceMinimum=0, TargetMaximum=1, TargetMinimum=0, Value=@n75.Result).Result
     B=&RigVMFunction_MathDoubleRemap(SourceMaximum=20, SourceMinimum=0, TargetMaximum=1, TargetMinimum=0, Value=@n76.Result).Result
    ).Result
   )
   &RigUnit_ParentConstraint(
    AdvancedSettings=(InterpolationType=Average,RotationOrderForFilter=XZY)
    Child=(Type=Bone,Name="deform_backpack")
    Filter=(TranslationFilter=(bX=false,bY=false,bZ=false),RotationFilter=(bX=true,bY=true,bZ=true),ScaleFilter=(bX=True,bY=True,bZ=True))
    Parents=((Item=(Type=Bone,Name="spine_01"),Weight=0.7),(Item=(Type=Bone,Name="spine_05"),Weight=1))
    Weight=0.75
    ExecutePin=@RigVMFunction_Sequence_1_1.E
   )
   > n77 = &RigUnit_ParentConstraint(
    AdvancedSettings=(InterpolationType=Average,RotationOrderForFilter=XZY)
    Child=(Type=Bone,Name="deform_backpack")
    Filter=(TranslationFilter=(bX=true,bY=true,bZ=true),RotationFilter=(bX=false,bY=false,bZ=false),ScaleFilter=(bX=True,bY=True,bZ=True))
    Parents=((Item=(Type=Bone,Name="spine_01"),Weight=0.5),(Item=(Type=Bone,Name="spine_05"),Weight=1))
    Weight=0.25
   )
   n78 = &RigVMFunction_Sequence(ExecuteContext=@n77.ExecutePin)
   n79 = &RigVMFunction_MathQuaternionToEuler(
    RotationOrder=ZYX
    Value=&RigUnit_GetRelativeTransformForItem(Child=(Type=Bone,Name="spine_05"), Parent=(Type=Bone,Name="pelvis"), bChildInitial=False, bParentInitial=false).RelativeTransform.Rotation
   )
   n80 = &RigVMFunction_AlphaInterp(
    InterpSpeedDecreasing=45
    InterpSpeedIncreasing=45
    Scale=1
    bClampResult=False
    Value=&RigVMFunction_MathDoubleRemap(SourceMaximum=50, SourceMinimum=-50, TargetMaximum=1, TargetMinimum=0, Value=@n79.Result.X).Result
   )
   &RigUnit_ModifyTransforms(
    ItemToModify=((Item=(Type=Bone,Name="deform_backpack"),Transform=(Rotation=(X=0.0347,Y=0.1045,Z=-0.0036,W=0.9939),Translation=(X=0,Y=0,Z=0),Scale3D=(X=1,Y=1,Z=1))))
    WeightMinimum=0
    Weight=@n80.Result
    ExecutePin=@n78.A
   )
   > &RigUnit_ModifyTransforms(
    ItemToModify=((Item=(Type=Bone,Name="deform_backpack"),Transform=(Rotation=(X=0,Y=0,Z=0.0436,W=0.999),Translation=(X=0,Y=-2,Z=0),Scale3D=(X=0.95,Y=1,Z=0.92))))
    WeightMinimum=0
    Weight=RigVMVariableNode() {variable=Skydive Thigh Relative}.Value
   )
   n81 = &RigVMFunction_AlphaInterp(
    InterpSpeedDecreasing=45
    InterpSpeedIncreasing=45
    Scale=1
    bClampResult=False
    Value=&RigVMFunction_MathDoubleRemap(SourceMaximum=-50, SourceMinimum=50, TargetMaximum=1, TargetMinimum=0, Value=@n79.Result.X).Result
   )
   &RigUnit_ModifyTransforms(
    ItemToModify=((Item=(Type=Bone,Name="deform_backpack"),Transform=(Rotation=(X=-0.0347,Y=-0.1045,Z=-0.0036,W=0.9939),Translation=(X=0,Y=0,Z=0),Scale3D=(X=1,Y=1,Z=1))))
    WeightMinimum=0
    Weight=@n81.Result
    ExecutePin=@n78.B
   )
   RigVMRerouteNode(Value=())
  ]}
]
hierarchy [
 root {loc=<0 0 0>, rot=<0 0 0 1>}
 pelvis {parent=root, loc=<0 -0.7321 92.4218>, rot=<0.0127 0.707 -0.0127 -0.707>}
 spine_01 {parent=pelvis, loc=<5.6002 -0.5128 0>, rot=<0 0 0.034 -0.9994>}
 spine_02 {parent=spine_01, loc=<8.0753 0.0139 0>, rot=<0 0 0.0238 -0.9997>}
 spine_03 {parent=spine_02, loc=<7.0393 0 0>, rot=<0 0 0.0149 -0.9999>}
 spine_04 {parent=spine_03, loc=<8.1024 0 0>, rot=<0 0 -0.0132 -0.9999>}
 spine_05 {parent=spine_04, loc=<10.0146 0 0>, rot=<0 0 -0.0015 -1>}
 attach_cape {parent=spine_05, loc=<9.1509 -13.3783 0>, rot=<0.0616 -0.7044 -0.0616 -0.7044>}
 attach_backpack {parent=spine_05, loc=<-0.123 -15.1423 0>, rot=<-0.0093 -0.707 0.0093 -0.707>}
 neck_01 {parent=spine_05, loc=<10.815 1.9196 0>, rot=<0 0 -0.1864 -0.9825>}
 neck_02 {parent=neck_01, loc=<4.7347 0.3045 0>, rot=<0 0 0.0363 -0.9993>}
 head {parent=neck_02, loc=<5.7632 0.5552 0>, rot=<0 0 0.115 -0.9934>}
 deform_head {parent=neck_02, loc=<5.7632 0.5552 0>, rot=<0 0 0.115 -0.9934>}
 clavicle_l {parent=spine_05, loc=<2.5856 -0.2866 -4.3777>, rot=<0.0465 -0.7289 -0.009 -0.6829>}
 upperarm_l {parent=clavicle_l, loc=<18.2264 0.1342 -0.001>, rot=<0.0253 -0.2693 0.0013 -0.9627>}
 lowerarm_l {parent=upperarm_l, loc=<29.0071 0.1686 -1.239>, rot=<0.0391 -0.0099 -0.2453 -0.9686>}
 hand_l {parent=lowerarm_l, loc=<25.8607 0.222 1.0274>, rot=<-0.7193 -0.0988 0.1193 -0.6773>}
 weapon_l {parent=hand_l, loc=<-0.1957 -3.4039 0.3727>, rot=<0 0 0 1>}
 thumb_01_l {parent=hand_l, loc=<1.3062 -0.3241 -2.4532>, rot=<0.6546 -0.4042 0.0664 -0.6354>}
 thumb_02_l {parent=thumb_01_l, loc=<5.9189 -0.0343 0.0318>, rot=<-0.0037 0.0213 0.298 -0.9543>}
 thumb_03_l {parent=thumb_02_l, loc=<2.2133 0.37 -0.0019>, rot=<0.0118 -0.013 0.0066 -0.9998>}
 index_metacarpal_l {parent=hand_l, loc=<1.6463 0.0115 -2.9522>, rot=<0.0158 -0.0124 -0.0173 -0.9996>}
 index_01_l {parent=index_metacarpal_l, loc=<8.263 0.0344 -0.0538>, rot=<-0.0044 0.0485 0.0924 -0.9945>}
 index_02_l {parent=index_01_l, loc=<3.3438 -0.3011 -0.002>, rot=<-0.002 0.0027 0.3225 -0.9466>}
 index_03_l {parent=index_02_l, loc=<2.5498 -0.1258 -0.0044>, rot=<-0.002 0.0067 0.2558 -0.9667>}
 middle_metacarpal_l {parent=hand_l, loc=<2.1758 -0.1739 -0.5976>, rot=<0.006 0.0168 -0.0329 -0.9993>}
 middle_01_l {parent=middle_metacarpal_l, loc=<7.294 -0.2437 0.0123>, rot=<-0.0208 0.0042 0.0858 -0.9961>}
 middle_02_l {parent=middle_01_l, loc=<3.996 0.2078 0.1134>, rot=<-0.0218 0.0258 0.3512 -0.9357>}
 middle_03_l {parent=middle_02_l, loc=<2.6009 -0.0136 0.1279>, rot=<0.0082 0.0188 0.1708 -0.9851>}
 ring_metacarpal_l {parent=hand_l, loc=<2.1973 -0.3593 1.3588>, rot=<-0.0284 0.0722 -0.0458 -0.9959>}
 ring_01_l {parent=ring_metacarpal_l, loc=<6.9956 -0.6101 -0.2032>, rot=<-0.0489 -0.0252 0.1892 -0.9804>}
 ring_02_l {parent=ring_01_l, loc=<3.6018 0.46 -0.0017>, rot=<0 0 0.2806 -0.9598>}
 ring_03_l {parent=ring_02_l, loc=<2.4098 0.1099 -0.0038>, rot=<0 0 0.1189 -0.9929>}
 pinky_metacarpal_l {parent=hand_l, loc=<1.7534 -0.9665 3.3696>, rot=<-0.0397 0.0814 -0.069 -0.9935>}
 pinky_01_l {parent=pinky_metacarpal_l, loc=<7.1386 -0.4785 -0.1076>, rot=<-0.0654 -0.0035 0.2242 -0.9723>}
 pinky_02_l {parent=pinky_01_l, loc=<2.8159 0.0856 -0.0018>, rot=<0.0037 0.0011 0.2854 -0.9584>}
 pinky_03_l {parent=pinky_02_l, loc=<2.2006 0.0274 -0.0035>, rot=<-0.0002 0.0098 0.1445 -0.9895>}
 hand_attach_l {parent=hand_l, loc=<8.4621 -2.996 -0.0083>, rot=<0.7071 0.7071 0 0>}
 deform_wrist_dn_base_l {parent=hand_l, loc=<0 -3.4173 0>, rot=<0 0 0 -1>}
 deform_wrist_dn_l {parent=deform_wrist_dn_base_l, loc=<0 0 0>, rot=<0 0 0 1>}
 deform_wrist_up_base_l {parent=hand_l, loc=<0 2.7805 0>, rot=<0 0 0 -1>}
 deform_wrist_up_l {parent=deform_wrist_up_base_l, loc=<0 0 0>, rot=<0 0 0 1>}
 lowerarm_twist_02_l {parent=lowerarm_l, loc=<9.6199 0 0>, rot=<0 0 0 -1>}
 lowerarm_twist_01_l {parent=lowerarm_l, loc=<17.8549 0 0>, rot=<0 0 0 -1>}
 deform_elbow_in_base_l {parent=lowerarm_l, loc=<0.2112 5.645 0.8274>, rot=<0 0 0 -1>}
 deform_elbow_in_l {parent=deform_elbow_in_base_l, loc=<0 0 0>, rot=<0 0 0 1>}
 deform_elbow_out_base_l {parent=lowerarm_l, loc=<0 -7.5129 0>, rot=<0 0 0 -1>}
 deform_elbow_out_l {parent=deform_elbow_out_base_l, loc=<0 0 0>, rot=<0 0 0 1>}
 deform_wrist_l {parent=lowerarm_l, loc=<22.0972 0.2618 1.1021>, rot=<-0.7657 -0.0058 0.0423 -0.6418>}
 upperarm_twist_01_l {parent=upperarm_l, loc=<7.9796 0.0843 -0.6195>, rot=<0 0 0 -1>}
 deform_upperarm_bk_base_l {parent=upperarm_twist_01_l, loc=<-6.8983 -13.3826 -1.3258>, rot=<0 0 0 1>}
 deform_upperarm_bk_l {parent=deform_upperarm_bk_base_l, loc=<0 0 0>, rot=<0 0 0 1>}
 deform_upperarm_fr_base_l {parent=upperarm_twist_01_l, loc=<-8.3487 10.7117 -2.1455>, rot=<0 0 0 1>}
 deform_upperarm_fr_l {parent=deform_upperarm_fr_base_l, loc=<0 0 0>, rot=<0 0 0 1>}
 deform_upperarm_up_base_l {parent=upperarm_twist_01_l, loc=<-11.2012 -1.2234 8.7784>, rot=<0 0 0 1>}
 deform_upperarm_up_l {parent=deform_upperarm_up_base_l, loc=<0 0 0>, rot=<0 0 0 1>}
 upperarm_twist_02_l {parent=upperarm_l, loc=<19.5059 0.0843 -0.6195>, rot=<0 0 0 -1>}
 deform_bicep_base_l {parent=upperarm_twist_02_l, loc=<0 8.0494 0>, rot=<0 0 0 1>}
 deform_bicep_l {parent=deform_bicep_base_l, loc=<0 0 0>, rot=<0 0 0 1>}
 deform_shoulderpad_l_base_root {parent=clavicle_l, loc=<18.2756 -0.8964 -0.0103>, rot=<0.0008 0.0336 -0.0252 -0.9991>}
 deform_shoulderpad_l_base {parent=deform_shoulderpad_l_base_root, loc=<1.4726 0 3.546>, rot=<0 0 0 1>}
 deform_shoulderpad_l {parent=deform_shoulderpad_l_base, loc=<0 0 0>, rot=<0 0 0 1>}
 deform_shoulderpad_target_l {parent=deform_shoulderpad_l, loc=<-2.381 -1.1601 15.217>, rot=<0 0 0 1>}
 deform_shoulderpad_target_side_l {parent=deform_shoulderpad_l, loc=<14.0512 -1.4386 -3.872>, rot=<0 0 0 1>}
 deform_shoulderpad_target_up_l {parent=deform_shoulderpad_l, loc=<-2.381 -1.1601 15.217>, rot=<0 0 0 1>}
 clavicle_r {parent=spine_05, loc=<2.5855 -0.2866 4.3777>, rot=<0.7289 0.0465 -0.6829 0.009>}
 upperarm_r {parent=clavicle_r, loc=<-18.2264 -0.1342 0.001>, rot=<0.0253 -0.2693 0.0013 -0.9627>}
 lowerarm_r {parent=upperarm_r, loc=<-29.0071 -0.1686 1.239>, rot=<0.0391 -0.0099 -0.2453 -0.9686>}
 hand_r {parent=lowerarm_r, loc=<-25.8607 -0.222 -1.0274>, rot=<-0.7193 -0.0988 0.1193 -0.6773>}
 weapon_r {parent=hand_r, loc=<-7.3765 3.1693 -1.5356>, rot=<0 0 0 1>}
 thumb_01_r {parent=hand_r, loc=<-1.3118 0.3191 2.4588>, rot=<0.6546 -0.4042 0.0664 -0.6354>}
 thumb_02_r {parent=thumb_01_r, loc=<-5.933 0.0448 -0.0402>, rot=<-0.0037 0.0213 0.298 -0.9543>}
 thumb_03_r {parent=thumb_02_r, loc=<-2.2221 -0.3682 0>, rot=<0.0118 -0.013 0.0066 -0.9998>}
 index_metacarpal_r {parent=hand_r, loc=<-1.6533 -0.0182 2.9598>, rot=<0.0158 -0.0124 -0.0173 -0.9996>}
 index_01_r {parent=index_metacarpal_r, loc=<-8.3004 -0.036 0.0525>, rot=<-0.0044 0.0485 0.0924 -0.9945>}
 index_02_r {parent=index_01_r, loc=<-3.3585 0.3004 0>, rot=<-0.002 0.0027 0.3225 -0.9466>}
 index_03_r {parent=index_02_r, loc=<-2.5577 0.123 0>, rot=<-0.002 0.0067 0.2558 -0.9667>}
 middle_metacarpal_r {parent=hand_r, loc=<-2.1856 0.1725 0.5984>, rot=<0.006 0.0168 -0.0329 -0.9993>}
 middle_01_r {parent=middle_metacarpal_r, loc=<-7.327 0.2444 -0.0128>, rot=<-0.0208 0.0042 0.0858 -0.9961>}
 middle_02_r {parent=middle_01_r, loc=<-4.0157 -0.1414 -0.0001>, rot=<-0.0218 0.0258 0.3512 -0.9357>}
 middle_03_r {parent=middle_02_r, loc=<-2.6103 -0.0516 -0.0001>, rot=<0.0082 0.0188 0.1708 -0.9851>}
 ring_metacarpal_r {parent=hand_r, loc=<-2.2071 0.3626 -1.3637>, rot=<-0.0284 0.0722 -0.0458 -0.9959>}
 ring_01_r {parent=ring_metacarpal_r, loc=<-7.0273 0.6131 0.2042>, rot=<-0.0489 -0.0252 0.1892 -0.9804>}
 ring_02_r {parent=ring_01_r, loc=<-3.6171 -0.4646 -0.0001>, rot=<0 0 0.2806 -0.9598>}
 ring_03_r {parent=ring_02_r, loc=<-2.4151 -0.114 -0.0001>, rot=<0 0 0.1189 -0.9929>}
 pinky_metacarpal_r {parent=hand_r, loc=<-1.7616 0.9752 -3.3811>, rot=<-0.0397 0.0814 -0.069 -0.9935>}
 pinky_01_r {parent=pinky_metacarpal_r, loc=<-7.1704 0.4829 0.1095>, rot=<-0.0654 -0.0035 0.2242 -0.9723>}
 pinky_02_r {parent=pinky_01_r, loc=<-2.8272 -0.0892 -0.0001>, rot=<0.0037 0.0011 0.2854 -0.9584>}
 pinky_03_r {parent=pinky_02_r, loc=<-2.2046 -0.0307 -0.0001>, rot=<-0.0002 0.0098 0.1445 -0.9895>}
 hand_attach_r {parent=hand_r, loc=<-8.4995 2.9995 0>, rot=<0 0 -0.7071 -0.7071>}
 deform_wrist_dn_base_r {parent=hand_r, loc=<0 3.4172 0>, rot=<0 0 0 -1>}
 deform_wrist_dn_r {parent=deform_wrist_dn_base_r, loc=<0 0 0>, rot=<0 0 0 1>}
 deform_wrist_up_base_r {parent=hand_r, loc=<-0.0001 -2.7804 0>, rot=<0 0 0 -1>}
 deform_wrist_up_r {parent=deform_wrist_up_base_r, loc=<0 0 0>, rot=<0 0 0 1>}
 lowerarm_twist_02_r {parent=lowerarm_r, loc=<-9.6199 0 0>, rot=<0 0 0 -1>}
 lowerarm_twist_01_r {parent=lowerarm_r, loc=<-17.8549 0 0>, rot=<0 0 0 -1>}
 deform_elbow_in_base_r {parent=lowerarm_r, loc=<-0.2114 -5.6449 -0.827>, rot=<0 0 0 -1>}
 deform_elbow_in_r {parent=deform_elbow_in_base_r, loc=<0 0 0>, rot=<0 0 0 1>}
 deform_elbow_out_base_r {parent=lowerarm_r, loc=<-0.0001 7.5129 0.0002>, rot=<0 0 0 -1>}
 deform_elbow_out_r {parent=deform_elbow_out_base_r, loc=<0 0 0>, rot=<0 0 0 1>}
 deform_wrist_r {parent=lowerarm_r, loc=<-22.1583 -0.2612 -1.1122>, rot=<-0.7657 -0.0058 0.0423 -0.6418>}
 upperarm_twist_01_r {parent=upperarm_r, loc=<-7.9796 -0.0843 0.6195>, rot=<0 0 0 -1>}
 deform_upperarm_bk_base_r {parent=upperarm_twist_01_r, loc=<6.8985 13.3826 1.3257>, rot=<0 0 0 1>}
 deform_upperarm_bk_r {parent=deform_upperarm_bk_base_r, loc=<0 0 0>, rot=<0 0 0 1>}
 deform_upperarm_fr_base_r {parent=upperarm_twist_01_r, loc=<8.3485 -10.7117 2.1456>, rot=<0 0 0 1>}
 deform_upperarm_fr_r {parent=deform_upperarm_fr_base_r, loc=<0 0 0>, rot=<0 0 0 1>}
 deform_upperarm_up_base_r {parent=upperarm_twist_01_r, loc=<11.2015 1.2234 -8.7788>, rot=<0 0 0 1>}
 deform_upperarm_up_r {parent=deform_upperarm_up_base_r, loc=<0 0 0>, rot=<0 0 0 -1>}
 upperarm_twist_02_r {parent=upperarm_r, loc=<-19.5059 -0.0843 0.6195>, rot=<0 0 0 -1>}
 deform_bicep_base_r {parent=upperarm_twist_02_r, loc=<0.0002 -8.0494 -0.0003>, rot=<0 0 0 1>}
 deform_bicep_r {parent=deform_bicep_base_r, loc=<0 0 0>, rot=<0 0 0 1>}
 deform_bandage_root {parent=upperarm_r, loc=<-14.7257 1.4684 1.4017>, rot=<-0.0055 -0.0857 -0.0121 -0.9962>}
 dyn_bandage_2 {parent=deform_bandage_root, loc=<0.9391 -0.073 -8.6434>, rot=<0.6061 -0.2329 0.6425 0.407>}
 dyn_bandage_1 {parent=deform_bandage_root, loc=<-0.4312 1.9867 -9.195>, rot=<0.8836 -0.2346 0.3979 0.0765>}
 deform_shoulderpad_r_base_root {parent=clavicle_r, loc=<-18.2756 0.8964 0.0102>, rot=<0.0008 0.0336 -0.0252 -0.9991>}
 deform_shoulderpad_r_base {parent=deform_shoulderpad_r_base_root, loc=<-1.4726 0 -3.5455>, rot=<0 0 0 1>}
 deform_shoulderpad_r {parent=deform_shoulderpad_r_base, loc=<0 0 0>, rot=<0 0 0 1>}
 deform_shoulderpad_target_r {parent=deform_shoulderpad_r, loc=<2.381 1.1601 -15.2174>, rot=<0 0 0 1>}
 deform_shoulderpad_target_up_r {parent=deform_shoulderpad_r, loc=<2.381 1.1601 -15.2174>, rot=<0 0 0 1>}
 deform_shoulderpad_target_side_r {parent=deform_shoulderpad_r, loc=<-14.0512 1.4386 3.8713>, rot=<0 0 0 1>}
 deform_backpack {parent=spine_05, loc=<1.9038 -15.191 0>, rot=<0 0 -0.0401 -0.9992>}
 deform_pec_aim_base_l {parent=spine_04, loc=<5.0232 13.6719 -2.6508>, rot=<0.0672 -0.5201 0.3228 -0.7879>}
 deform_pec_aim_l {parent=deform_pec_aim_base_l, loc=<0 0 0>, rot=<0 0 0 1>}
 deform_pec_base_l {parent=deform_pec_aim_l, loc=<16.3279 0 0>, rot=<0 0 0 1>}
 deform_pec_l {parent=deform_pec_base_l, loc=<0 0 0>, rot=<0 0 0 1>}
 deform_pec_aim_base_r {parent=spine_04, loc=<5.023 13.6718 2.6508>, rot=<-0.5201 -0.0672 0.7879 0.3228>}
 deform_pec_aim_r {parent=deform_pec_aim_base_r, loc=<0 0 0>, rot=<0 0 0 1>}
 deform_pec_base_r {parent=deform_pec_aim_r, loc=<-16.3279 0 0>, rot=<0 0 0 -1>}
 deform_pec_r {parent=deform_pec_base_r, loc=<0 0 0.0002>, rot=<0 0 0 -1>}
 deform_lat_l_base {parent=spine_03, loc=<2.5566 -10.3716 -17.6918>, rot=<0 0 0 1>}
 deform_lat_l {parent=deform_lat_l_base, loc=<0 0 0>, rot=<0 0 0 1>}
 deform_lat_r_base {parent=spine_03, loc=<2.5568 -10.3716 17.6918>, rot=<0 0 1 0>}
 deform_lat_r {parent=deform_lat_r_base, loc=<0 0 0>, rot=<0 0 0 -1>}
 thigh_l {parent=pelvis, loc=<-1.8716 -0.0109 -11.9936>, rot=<0.0012 0.0572 0.0155 -0.9982>}
 calf_l {parent=thigh_l, loc=<-41.9178 -0.2246 0.0926>, rot=<0.0006 -0.0161 -0.0391 -0.9991>}
 foot_l {parent=calf_l, loc=<-38.7057 1.2438 0.2452>, rot=<0.0975 -0.0474 0.0317 -0.9936>}
 ball_l {parent=foot_l, loc=<-6.1614 13.1469 0.6807>, rot=<0.0001 -0.0001 -0.7063 -0.7079>}
 deform_ankle_l {parent=foot_l, loc=<-1.7225 2.6679 -8.9872>, rot=<-0.1011 0.0023 0.0053 -0.9949>}
 calf_twist_02_l {parent=calf_l, loc=<-15.9268 -0.0967 0.0397>, rot=<0.0237 0.0095 -0.0035 -0.9997>}
 calf_twist_01_l {parent=calf_l, loc=<-32.9335 -0.0967 0.0397>, rot=<0.0237 0.0095 -0.0035 -0.9997>}
 deform_knee_in_base_l {parent=calf_l, loc=<1.298 -8.0356 -0.0496>, rot=<0 0 0 -1>}
 deform_knee_in_l {parent=deform_knee_in_base_l, loc=<0 0 0>, rot=<0 0 0 1>}
 deform_knee_out_base_l {parent=calf_l, loc=<0 6.8739 0>, rot=<0 0 0 -1>}
 deform_knee_out_l {parent=deform_knee_out_base_l, loc=<0 0 0>, rot=<0 0 0 1>}
 dyn_lever_l {parent=calf_l, loc=<-29.9186 6.1385 0.378>, rot=<-0.6358 0.2828 0.6546 0.2953>}
 deform_kneeshield_l {parent=calf_l, loc=<4.0984 13.6572 1.3711>, rot=<-0.0006 0.019 0.0415 -0.999>}
 thigh_twist_01_l {parent=thigh_l, loc=<-21.6303 0 0>, rot=<0.0013 0 0 -1>}
 deform_groin_base_l {parent=thigh_l, loc=<-2.1908 8.5661 -1.7442>, rot=<0 0 0 -1>}
 deform_groin_l {parent=deform_groin_base_l, loc=<0 0 0>, rot=<0 0 0 1>}
 deform_thighpad_base_l {parent=thigh_l, loc=<-3.6342 -1.1529 -4.2228>, rot=<0.0572 -0.0002 0.9984 0.0024>}
 deform_thighpad_l {parent=deform_thighpad_base_l, loc=<-9.6317 0 -2.5732>, rot=<0 0 0 1>}
 thigh_r {parent=pelvis, loc=<-1.8716 -0.0109 11.9936>, rot=<0.0572 -0.0012 0.9982 0.0155>}
 calf_r {parent=thigh_r, loc=<41.9178 0.2246 -0.0926>, rot=<0.0006 -0.0161 -0.0391 -0.9991>}
 foot_r {parent=calf_r, loc=<38.7057 -1.2438 -0.2452>, rot=<0.0975 -0.0474 0.0317 -0.9936>}
 ball_r {parent=foot_r, loc=<6.1614 -13.1469 -0.6807>, rot=<0.0001 -0.0001 -0.7063 -0.7079>}
 deform_ankle_r {parent=foot_r, loc=<1.7225 -2.6678 8.9872>, rot=<-0.1011 0.0023 0.0053 -0.9949>}
 calf_twist_02_r {parent=calf_r, loc=<15.9268 0.0967 -0.0397>, rot=<0.0237 0.0095 -0.0035 -0.9997>}
 calf_twist_01_r {parent=calf_r, loc=<32.9335 0.0967 -0.0397>, rot=<0.0237 0.0095 -0.0035 -0.9997>}
 deform_knee_in_base_r {parent=calf_r, loc=<-1.2981 8.0356 0.0496>, rot=<0 0 0 1>}
 deform_knee_in_r {parent=deform_knee_in_base_r, loc=<0 0 0>, rot=<0 0 0 1>}
 deform_knee_out_base_r {parent=calf_r, loc=<0 -6.8739 0>, rot=<0 0 0 1>}
 deform_knee_out_r {parent=deform_knee_out_base_r, loc=<0 0 0>, rot=<0 0 0 1>}
 dyn_lever_r {parent=calf_r, loc=<29.9186 -6.1385 -0.378>, rot=<-0.6358 0.2828 0.6546 0.2953>}
 deform_kneeshield_r {parent=calf_r, loc=<-4.0984 -13.6572 -1.3706>, rot=<-0.0006 0.019 0.0415 -0.999>}
 thigh_twist_01_r {parent=thigh_r, loc=<21.6303 0 0>, rot=<0.0013 0 0 -1>}
 deform_groin_base_r {parent=thigh_r, loc=<2.1908 -8.5661 1.7442>, rot=<0 0 0 -1>}
 deform_groin_r {parent=deform_groin_base_r, loc=<0 0 0>, rot=<0 0 0 1>}
 deform_thighpad_base_r {parent=thigh_r, loc=<3.6342 1.153 4.2227>, rot=<0.0572 -0.0002 0.9984 0.0024>}
 deform_thighpad_r {parent=deform_thighpad_base_r, loc=<9.6317 0 2.5733>, rot=<0 0 0 1>}
 deform_glute_base_l {parent=pelvis, loc=<-10.1877 -11.6217 -12.899>, rot=<0.0275 0.0501 -0.4472 -0.8926>}
 deform_glute_l {parent=deform_glute_base_l, loc=<0 0 0>, rot=<0 0 0 1>}
 deform_glute_base_r {parent=pelvis, loc=<-10.1877 -11.6218 12.899>, rot=<0.0501 -0.0275 0.8926 -0.4472>}
 deform_glute_r {parent=deform_glute_base_r, loc=<0 0 0>, rot=<0 0 0 1>}
 deform_buckler_f {parent=pelvis, loc=<2.2232 13.4596 0.0002>, rot=<-0.0127 -0.707 0.0127 -0.707>}
 deform_buckler_b {parent=pelvis, loc=<1.2452 -13.8442 -0.0004>, rot=<-0.0127 -0.707 0.0127 -0.707>}
 upperarm_twist_02_r_null {type=Null, loc=<0 0 0>, rot=<0 0 0 1>}
 lowerarm_r_null {type=Null, loc=<0 0 0>, rot=<0 0 0 1>}
]
variables [
 Crouching Enabled {type=double, default=0}
 DBNO disabled {type=double, default=0}
 Lunch Emote {type=bool, default=False}
 Skydive Enabled {type=double, default=0}
 Skydive Thigh Relative {type=double, default=0}
 isDBNOCarried {type=bool, default=False}
]
