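# Physics asset for the large male RadiumFox Flame player body cosmetic, covering only the calves, upper arms, upper spine and two dyn_lever bones.
# Capsules approximate the limbs and spine_05, while boxes shape the levers, which appear to be dynamic accessory pieces.
# Each lever is constrained to its calf and each upper arm to spine_05.
# Only the right lever restricts rotation, locking both swings with a narrow twist limit; the left lever and arm joints rotate freely.
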PhysicsAsset /BRCosmetics/Characters/Player/Male/Large/Bodies/M_LRG_RadiumFox_Flame/Meshes/M_LRG_RadiumFox_Flame_Physics
bodies {
 calf_l [{
   shapeType=Capsule
   length=35.845
   radius=4.4951
   shapeName=calf_l_capsule
   center=<-20.5013 3.2199 -0.1513>
   rotation=<rot 1.0422 91.6364 -86.3616>
  }]
 calf_r [{
   shapeType=Capsule
   length=36.2606
   radius=4.737
   shapeName=calf_r_capsule
   center=<20.8875 -2.9981 0.1848>
   rotation=<rot 1.4355 90.8011 -88.0216>
  }]
 dyn_lever_l [{
   shapeType=Box
   extentX=2.2558
   extentY=6.3389
   extentZ=11.983
   shapeName=dyn_lever_r_box
   center=<0.0322 -13.5029 2.1144>
   rotation=<rot -3.8471 -89.8833 -89.5591>
  }]
 dyn_lever_r [{
   shapeType=Box
   extentX=2.2558
   extentY=6.3389
   extentZ=11.983
   shapeName=dyn_lever_r_box
   center=<-0.068 13.5336 -1.945>
   rotation=<rot -3.8471 -89.8833 -89.5591>
  }]
 spine_05 [{shapeType=Capsule, length=19.9107, radius=13.7454, shapeName=spine_05_capsule, center=<2.7155 -11.0146 0>, rotation=<rot 0 -69.6223 -90>}]
 upperarm_l [{
   shapeType=Capsule
   length=28.5603
   radius=6.5383
   shapeName=upperarm_l_capsule
   center=<7.6103 -2.3028 0.0949>
   rotation=<rot -0.0001 -88.5024 -88.1234>
  }]
 upperarm_r [{
   shapeType=Capsule
   length=26.1774
   radius=6.2698
   shapeName=upperarm_r_capsule
   center=<-7.9135 2.537 -0.0064>
   rotation=<rot -0.0001 -89.5447 -85.9642>
  }]
}
constraints [
 {
  bone1Name=dyn_lever_r
  bone2Name=calf_r
  swing1LimitDegrees=45
  swing1Motion=Locked
  swing2LimitDegrees=45
  swing2Motion=Locked
  twistLimitDegrees=17
 }
 {
  bone1Name=upperarm_l
  bone2Name=spine_05
  swing1LimitDegrees=45
  swing1Motion=Free
  swing2LimitDegrees=45
  swing2Motion=Free
  twistLimitDegrees=45
  twistMotion=Free
 }
 {
  bone1Name=upperarm_r
  bone2Name=spine_05
  swing1LimitDegrees=45
  swing1Motion=Free
  swing2LimitDegrees=45
  swing2Motion=Free
  twistLimitDegrees=45
  twistMotion=Free
 }
 {
  bone1Name=dyn_lever_l
  bone2Name=calf_l
  swing1LimitDegrees=45
  swing1Motion=Free
  swing2LimitDegrees=45
  swing2Motion=Free
  twistLimitDegrees=45
  twistMotion=Free
 }
]
