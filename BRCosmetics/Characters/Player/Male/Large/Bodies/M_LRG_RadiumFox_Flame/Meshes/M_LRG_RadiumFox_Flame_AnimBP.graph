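# Post-copy anim Blueprint for the male RadiumFox Flame cosmetic body, adding rig correction and secondary motion to the attached parent's pose.
# The copied pose runs through the body's Control Rig, fed DBNO-carried and Lunch Emote flags, then spring-based Anim Dynamics on lever and bandage bones.
# The event graph sets Lunch Emote when the owner's active montage is a specific Builder emote.
# Shoulder-pad and bandage bone edits, RigidBody and several movement-state checks are unconnected leftovers.
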AnimBlueprint /BRCosmetics/Characters/Player/Male/Large/Bodies/M_LRG_RadiumFox_Flame/Meshes/M_LRG_RadiumFox_Flame_AnimBP
target_skeleton=&/M_LRG_RadiumFox_Flame_Skeleton
graphs [
 AnimGraph {nodes [
   n1 = "Transform (Modify) Bone - Bone: shoulderpads_r"!(
    ComponentPose
    bAlphaBoolEnabled:bool=Get bIsParachuteOpened()
    Translation="-2.287,2.043,-0.695"
    Rotation="6.383,-8.531,-14.020"
    Scale="0.75,0.75,0.75"
   ) -> (Pose) {
    alphaInputType=Bool
    rotationMode=BMM_Additive
    scaleMode=BMM_Replace
    translationMode=BMM_Additive
    translationSpace=BCS_BoneSpace
    alphaBoolBlend {blendInTime=2, blendOption=Sinusoidal, blendOutTime=2}
    alphaScaleBiasClamp {bInterpResult=true, interpSpeedDecreasing=1, interpSpeedIncreasing=1}
    boneToModify {boneName=shoulderpads_r}
    rotation=<rot 6.383 -8.531 -14.02>
    scale=<0.75 0.75 0.75>
    translation=<-2.287 2.043 -0.695>
   }
   n2 = "Transform (Modify) Bone - Bone: shoulderpads_r"(ComponentPose=@n1, bAlphaBoolEnabled=True, Translation="-8.081,0.085,3.391", Rotation="18.114,0.404,14.964", Scale="0.75,0.75,0.75") {
    alphaInputType=Bool
    rotationMode=BMM_Additive
    scaleMode=BMM_Replace
    translationMode=BMM_Additive
    translationSpace=BCS_BoneSpace
    alphaBoolBlend {blendInTime=2, blendOption=Sinusoidal, blendOutTime=2}
    alphaScaleBiasClamp {bInterpResult=true, interpSpeedDecreasing=1, interpSpeedIncreasing=1}
    boneToModify {boneName=shoulderpads_r}
    rotation=<rot 18.114 0.404 14.964>
    scale=<0.75 0.75 0.75>
    translation=<-8.0811 0.085 3.3908>
   }
   n3 = "Transform (Modify) Bone - Bone: shoulderpads_r"(ComponentPose=@n2, bAlphaBoolEnabled=True, Translation="1.763,2.106,0.420", Rotation="-25.910,23.689,-3.918", Scale="0.9,0.9,0.9") {
    alphaInputType=Bool
    rotationMode=BMM_Additive
    scaleMode=BMM_Replace
    translationMode=BMM_Additive
    translationSpace=BCS_BoneSpace
    alphaBoolBlend {blendInTime=2, blendOption=Sinusoidal, blendOutTime=2}
    alphaScaleBiasClamp {bInterpResult=true, interpSpeedDecreasing=1, interpSpeedIncreasing=1}
    boneToModify {boneName=shoulderpads_r}
    rotation=<rot -25.91 23.689 -3.918>
    scale=<0.9 0.9 0.9>
    translation=<1.763 2.106 0.42>
   }
   n4 = "Transform (Modify) Bone - Bone: shoulderpads_r"(ComponentPose=@n3, bAlphaBoolEnabled=True, Translation="-15,0,5", Rotation="35,0,0", Scale="0.9,0.9,0.9") {
    alphaInputType=Bool
    rotationMode=BMM_Additive
    scaleMode=BMM_Replace
    translationMode=BMM_Additive
    translationSpace=BCS_BoneSpace
    alphaBoolBlend {blendInTime=1, blendOption=Sinusoidal, blendOutTime=1.5}
    alphaScaleBiasClamp {bInterpResult=true, interpSpeedDecreasing=1, interpSpeedIncreasing=1}
    boneToModify {boneName=shoulderpads_r}
    rotation {pitch=35}
    scale=<0.9 0.9 0.9>
    translation {x=-15, z=5}
   }
   n5 = "Transform (Modify) Bone - Bone: shoulderpads_r"(
    ComponentPose=@n4
    bAlphaBoolEnabled=Get bIsParachuteOpened()
    Translation="-2.287,2.043,-0.695"
    Rotation="6.383,-8.531,-14.020"
    Scale="0.75,0.75,0.75"
   ) {
    alphaInputType=Bool
    rotationMode=BMM_Additive
    scaleMode=BMM_Replace
    translationMode=BMM_Additive
    translationSpace=BCS_BoneSpace
    alphaBoolBlend {blendInTime=2, blendOption=Sinusoidal, blendOutTime=2}
    alphaScaleBiasClamp {bInterpResult=true, interpSpeedDecreasing=1, interpSpeedIncreasing=1}
    boneToModify {boneName=shoulderpads_r}
    rotation=<rot 6.383 -8.531 -14.02>
    scale=<0.75 0.75 0.75>
    translation=<-2.287 2.043 -0.695>
   }
   n6 = "Transform (Modify) Bone - Bone: shoulderpads_r"(ComponentPose=@n5, bAlphaBoolEnabled=True, Translation="-8.081,0.085,3.391", Rotation="18.114,0.404,14.964", Scale="0.75,0.75,0.75") {
    alphaInputType=Bool
    rotationMode=BMM_Additive
    scaleMode=BMM_Replace
    translationMode=BMM_Additive
    translationSpace=BCS_BoneSpace
    alphaBoolBlend {blendInTime=2, blendOption=Sinusoidal, blendOutTime=2}
    alphaScaleBiasClamp {bInterpResult=true, interpSpeedDecreasing=1, interpSpeedIncreasing=1}
    boneToModify {boneName=shoulderpads_r}
    rotation=<rot 18.114 0.404 14.964>
    scale=<0.75 0.75 0.75>
    translation=<-8.0811 0.085 3.3908>
   }
   n7 = "Transform (Modify) Bone - Bone: shoulderpads_r"(ComponentPose=@n6, bAlphaBoolEnabled=True, Translation="1.763,2.106,0.420", Rotation="-25.910,23.689,-3.918", Scale="0.9,0.9,0.9") {
    alphaInputType=Bool
    rotationMode=BMM_Additive
    scaleMode=BMM_Replace
    translationMode=BMM_Additive
    translationSpace=BCS_BoneSpace
    alphaBoolBlend {blendInTime=2, blendOption=Sinusoidal, blendOutTime=2}
    alphaScaleBiasClamp {bInterpResult=true, interpSpeedDecreasing=1, interpSpeedIncreasing=1}
    boneToModify {boneName=shoulderpads_r}
    rotation=<rot -25.91 23.689 -3.918>
    scale=<0.9 0.9 0.9>
    translation=<1.763 2.106 0.42>
   }
   n8 = "Transform (Modify) Bone - Bone: shoulderpads_r"(ComponentPose=@n7, bAlphaBoolEnabled=True, Translation="-15,0,5", Rotation="35,0,0", Scale="0.9,0.9,0.9") {
    alphaInputType=Bool
    rotationMode=BMM_Additive
    scaleMode=BMM_Replace
    translationMode=BMM_Additive
    translationSpace=BCS_BoneSpace
    alphaBoolBlend {blendInTime=1, blendOption=Sinusoidal, blendOutTime=1.5}
    alphaScaleBiasClamp {bInterpResult=true, interpSpeedDecreasing=1, interpSpeedIncreasing=1}
    boneToModify {boneName=shoulderpads_r}
    rotation {pitch=35}
    scale=<0.9 0.9 0.9>
    translation {x=-15, z=5}
   }
   n9 = "Transform (Modify) Bone - Bone: dyn_bandage_1"!(ComponentPose=@n8, bAlphaBoolEnabled:bool=True, Translation="0,0,0", Rotation="0,0,0", Scale="0.05,0.05,0.05") -> (Pose) {
    alphaInputType=Bool
    rotationMode=BMM_Additive
    scaleMode=BMM_Replace
    translationMode=BMM_Additive
    translationSpace=BCS_BoneSpace
    alphaBoolBlend {blendInTime=2, blendOption=Sinusoidal, blendOutTime=2}
    alphaScaleBiasClamp {bInterpResult=true, interpSpeedDecreasing=1, interpSpeedIncreasing=1}
    boneToModify {boneName=dyn_bandage_1}
    scale=<0.05 0.05 0.05>
   }
   "Transform (Modify) Bone - Bone: dyn_bandage_2"!(
    ComponentPose=@n9
    bAlphaBoolEnabled:bool=Reroute Node(InputPin:bool) -> (OutputPin:bool)
    Translation="0,0,0"
    Rotation="0,0,0"
    Scale="0.05,0.05,0.05"
   ) -> (Pose) {
    alphaInputType=Bool
    rotationMode=BMM_Additive
    scaleMode=BMM_Replace
    translationMode=BMM_Additive
    translationSpace=BCS_BoneSpace
    alphaBoolBlend {blendInTime=2, blendOption=Sinusoidal, blendOutTime=2}
    alphaScaleBiasClamp {bInterpResult=true, interpSpeedDecreasing=1, interpSpeedIncreasing=1}
    boneToModify {boneName=dyn_bandage_2}
    scale=<0.05 0.05 0.05>
   }
   RigidBody() {
    bOverrideWorldGravity=true
    overridePhysicsAsset=&/M_LRG_RadiumFox_Flame_Physics
    simulationSpace=BaseBoneSpace
    alphaBoolBlend {blendOption=Linear}
    baseBoneRef {boneName=spine_05}
    componentAppliedLinearAccClamp=<5000 5000 5000>
    componentLinearAccScale=<0.7 0.7 0.7>
    componentLinearVelScale=<0.7 0.7 0.7>
    overrideWorldGravity {z=-980}
    simSpaceSettings {maxAngularAcceleration=450, maxAngularVelocity=35, maxLinearAcceleration=150, maxLinearVelocity=10, velocityScaleZ=0, worldAlpha=1}
   }
   n10 = Control Rig!(
    Source=Copy Pose From Mesh() {bUseAttachedParent=true}
    Alpha
    isDBNOCarried:bool=Get bIsDBNOCarried()
    Lunch Emote:bool=Get Lunch Emote() -> (Lunch Emote:bool)
   ) -> (Pose) {
    alphaBoolBlend {blendOption=Linear}
    controlRigAssetReference {blueprintRigClass=&/M_LRG_RadiumFox_Flame_CtrlRig.M_LRG_RadiumFox_Flame_CtrlRig_C}
   }
   n11 = Anim Dynamics - Bone: dyn_lever_r(ComponentPose=Local To Component(LocalPose=@n10), Alpha=1, ExternalForce="0,0,0") {
    angularDampingOverride=0.9999
    angularSpringConstant=2000
    bAngularSpring=true
    bGravityOverrideInSimSpace=true
    bOverrideAngularDamping=true
    gravityScale=0.75
    simulationSpace=RootRelative
    alphaBoolBlend {blendOption=Linear}
    boundBone {boneName=dyn_lever_r}
    componentAppliedLinearAccClamp=<2500 2500 2500>
    componentLinearAccScale=<1 1 1>
    componentLinearVelScale=<1 1 1>
    gravityOverride {z=-980}
    physicsBodyDefinitions [{
      boundBone {boneName=dyn_lever_r}
      boxExtents=<10 10 10>
      constraintSetup {angularTargetAxis=AxisY, twistAxis=AxisY, angularLimitsMax=<15 0 0>, angularLimitsMin=<-10 0 0>, angularTarget=<0 0 0>}
      localJointOffset=<0 1 0>
     }]
   }
   n12 = Anim Dynamics - Bone: dyn_lever_l(ComponentPose=@n11, Alpha=1, ExternalForce="0,0,0") {
    angularDampingOverride=0.9999
    angularSpringConstant=2000
    bAngularSpring=true
    bGravityOverrideInSimSpace=true
    bOverrideAngularDamping=true
    gravityScale=0.75
    simulationSpace=RootRelative
    alphaBoolBlend {blendOption=Linear}
    boundBone {boneName=dyn_lever_l}
    componentAppliedLinearAccClamp=<2500 2500 2500>
    componentLinearAccScale=<1 1 1>
    componentLinearVelScale=<1 1 1>
    gravityOverride {z=-980}
    physicsBodyDefinitions [{
      boundBone {boneName=dyn_lever_l}
      boxExtents=<10 10 10>
      constraintSetup {angularTargetAxis=AxisY, twistAxis=AxisY, angularLimitsMax=<15 0 0>, angularLimitsMin=<-10 0 0>, angularTarget=<0 1 0>}
      localJointOffset=<0 -1 0>
     }]
   }
   n13 = Anim Dynamics - Bone: dyn_bandage_1(ComponentPose=@n12, Alpha=0.5, ExternalForce="0,0,0") {
    angularDampingOverride=0.9999
    angularSpringConstant=2500
    bAngularSpring=true
    bGravityOverrideInSimSpace=true
    bOverrideAngularDamping=true
    gravityScale=0.6
    simulationSpace=RootRelative
    alphaBoolBlend {blendOption=Linear}
    boundBone {boneName=dyn_bandage_1}
    componentAppliedLinearAccClamp=<2500 2500 2500>
    componentLinearAccScale=<1 1 1>
    componentLinearVelScale=<1 1 1>
    gravityOverride {z=-980}
    physicsBodyDefinitions [{
      boundBone {boneName=dyn_bandage_1}
      boxExtents=<10 10 10>
      constraintSetup {angularTargetAxis=AxisZ, angularLimitsMax=<5 15 35>, angularLimitsMin=<-5 -10 -20>, angularTarget=<0.4 1 1>}
      localJointOffset=<-0.4 -0.4 -0.4>
     }]
   }
   n14 = Anim Dynamics - Bone: dyn_bandage_2(ComponentPose=@n13, Alpha=0.5, ExternalForce="0,0,0") {
    angularDampingOverride=0.9999
    angularSpringConstant=2500
    bAngularSpring=true
    bGravityOverrideInSimSpace=true
    bOverrideAngularDamping=true
    gravityScale=0.6
    simulationSpace=RootRelative
    alphaBoolBlend {blendOption=Linear}
    boundBone {boneName=dyn_bandage_2}
    componentAppliedLinearAccClamp=<2500 2500 2500>
    componentLinearAccScale=<1 1 1>
    componentLinearVelScale=<1 1 1>
    gravityOverride {z=-980}
    physicsBodyDefinitions [{
      boundBone {boneName=dyn_bandage_2}
      boxExtents=<10 10 10>
      constraintSetup {angularTargetAxis=AxisZ, angularLimitsMax=<5 35 15>, angularLimitsMin=<-5 -1 -15>, angularTarget=<0.4 1 1>}
      localJointOffset=<-0.4 -0.4 -0.4>
     }]
   }
   Output Pose(Result=Component To Local(ComponentPose=@n14)) {layerGroup=None}
   Reroute Node(InputPin:Wildcard) -> (OutputPin:Wildcard)
   Reroute Node()
  ]}
 EventGraph {nodes [
   Event BlueprintUpdateAnimation()
   > n1 = Sequence!(execute) -> (then_0, then_1, then_2:Exec, then_3:Exec)
   n2 = GetAnimInstance(self=GetOwnerSkeletalMeshComponent())
   > Set Lunch Emote(Lunch Emote:bool="Equal (Object)"(
     A=GetCurrentActiveMontage!(self:Anim Instance Object Reference=@n2) -> (ReturnValue:Anim Montage Object Reference)
     B=/BRCosmetics/Animation/Game/MainPlayer/Emotes/Builder/Emote_Builder_CMM_M
    )) -> (Output_Get:bool)
   Reroute Node(InputPin:Exec=@n1.then_1) -> (OutputPin:Exec)
   Reroute Node(InputPin=@n1.then_2)
   Reroute Node(InputPin=@n1.then_3)
   Reroute Node()
   > Reroute Node()
   "To Float (Boolean)"(InBool=OR Boolean(A=Get bIsCrouchSprinting(), B=Get bIsCrouching()))
   n3 = Get bIsTacticalSprinting()
   "To Float (Boolean)"(InBool=@n3)
   "To Float (Boolean)"(InBool=Get bIsSkydiveFloating())
   "To Float (Boolean)"(InBool=Get bIsParachuteOpened())
   "To Float (Boolean)"(InBool=Get bIsBaseballSliding())
   "To Float (Boolean)"(InBool=Get bIsPlayingMeleeAnim())
   OR Boolean!(A=Get bIsSkydiveFloating(), B=Get bIsSkydiveDiveMode(), C:bool=Get bIsParachuteOpened()) -> (ReturnValue)
   AND Boolean(A=Get bIsJumping(), B=Get bIsTacticalSprinting())
   OR Boolean!(A=Get bIsSprinting(), B=@n3) -> (ReturnValue)
   Branch(Condition=OR Boolean(A=false, B=false))
   Get bIsDBNO()
  ]}
]
variables [Lunch Emote {type=boolean}]
