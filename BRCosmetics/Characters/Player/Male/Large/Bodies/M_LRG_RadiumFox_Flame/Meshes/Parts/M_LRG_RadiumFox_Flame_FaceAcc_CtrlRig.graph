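# Control Rig for the RadiumFox Flame face accessory, applying procedural correctives to the deform head, neck and side pipe bones.
# Outside Front End, filtered parent constraints and head yaw adjust head and neck.
# Spine and pelvis spherical pose readers set Skydive Enabled, which, with a head pose reader, weights a neck offset.
# Head-relative rotation drives pipes when Not Rigid Body is set; clavicle pitch squashes the head, and pelvis height drives a Lunch Emote offset.
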ControlRigBlueprint /BRCosmetics/Characters/Player/Male/Large/Bodies/M_LRG_RadiumFox_Flame/Meshes/Parts/M_LRG_RadiumFox_Flame_FaceAcc_CtrlRig
graphs [
 RigVMFunctionLibrary {links [], nodes []}
 RigVMFunction_Sequence_1_ContainedGraph {nodes [
   n1 = RigVMFunctionEntryNode()
   RigVMFunction_Sequence = &RigVMFunction_Sequence(ExecuteContext=@n1.ExecuteContext)
   n2 = &RigVMFunction_Sequence(ExecuteContext=@RigVMFunction_Sequence.B)
   n3 = &RigVMFunction_Sequence(ExecuteContext=@n2.B)
   n4 = &RigVMFunction_Sequence(ExecuteContext=@n3.B)
   n5 = &RigVMFunction_Sequence(ExecuteContext=@n4.B)
   RigVMFunctionReturnNode(A=@RigVMFunction_Sequence.A, B=@n2.A, C=@n3.A, D=@n4.A, E=@n5.A, F=@n5.B)
  ]}
 RigVMModel {nodes [
   RigUnit_BeginExecution = &RigUnit_BeginExecution()
   RigVMFunction_Sequence = RigVMAggregateNode(ExecuteContext=@RigUnit_BeginExecution.ExecutePin)
   n1 = &RigVMFunction_ControlFlowBranch(ExecuteContext=@RigVMFunction_Sequence.A, Condition=RigVMVariableNode() {variable=Front End}.Value)
   &RigUnit_ParentConstraint(
    AdvancedSettings=(InterpolationType=Average,RotationOrderForFilter=XZY)
    Child=(Type=Bone,Name="deform_head")
    Filter=(TranslationFilter=(bX=false,bY=false,bZ=false),RotationFilter=(bX=true,bY=true,bZ=false),ScaleFilter=(bX=false,bY=false,bZ=false))
    Parents=((Item=(Type=Bone,Name="head"),Weight=1))
    Weight=0.4
    ExecutePin=@n1.False
   )
   > &RigUnit_ParentConstraint(
    AdvancedSettings=(InterpolationType=Average,RotationOrderForFilter=XZY)
    Child=(Type=Bone,Name="neck_01")
    Filter=(TranslationFilter=(bX=false,bY=false,bZ=false),RotationFilter=(bX=true,bY=true,bZ=false),ScaleFilter=(bX=false,bY=false,bZ=false))
    Parents=((Item=(Type=Bone,Name="spine_05"),Weight=1))
    Weight=0.15
   )
   > &RigUnit_ParentConstraint(
    AdvancedSettings=(InterpolationType=Average,RotationOrderForFilter=XZY)
    Child=(Type=Bone,Name="neck_02")
    Filter=(TranslationFilter=(bX=false,bY=false,bZ=false),RotationFilter=(bX=true,bY=true,bZ=false),ScaleFilter=(bX=false,bY=false,bZ=false))
    Parents=((Item=(Type=Bone,Name="spine_05"),Weight=1))
    Weight=0.15
   )
   > Get Delta Ref_2 = RigVMFunctionReferenceNode(Item=(Type=Bone,Name="head"))
   > &RigUnit_ModifyTransforms(
    ItemToModify=((Item=(Type=Bone,Name="deform_head"),Transform=(Rotation=(X=0,Y=0,Z=0.342,W=0.9397),Translation=(X=0,Y=0,Z=0),Scale3D=(X=1,Y=1,Z=1))))
    WeightMinimum=0
    Weight=&RigVMFunction_MathDoubleRemap(SourceMaximum=50, SourceMinimum=0, TargetMaximum=1, TargetMinimum=0, Value=@"Get Delta Ref_2".Result.Rotation.Yaw).Result
   )
   n2 = &RigUnit_SphericalPoseReader(
    ActiveRegionScaleFactors=(PositiveWidth=1,NegativeWidth=1,PositiveHeight=1,NegativeHeight=1)
    ActiveRegionSize=0.12
    Debug=(bDrawDebug=false,bDraw2D=False,bDrawLocalAxes=False,DebugScale=25,DebugSegments=20,DebugThickness=0.25)
    DriverAxis=(X=0,Y=1,Z=0)
    DriverItem=(Type=Bone,Name="spine_03")
    FalloffRegionScaleFactors=(PositiveWidth=1,NegativeWidth=1,PositiveHeight=1,NegativeHeight=1)
    FalloffSize=0.2
    FlipHeightScaling=False
    FlipWidthScaling=False
    OptionalParentItem=(Type=Bone,Name="root")
    RotationOffset=(X=10,Y=90,Z=0)
    ExecutePin=@RigVMFunction_Sequence.C
   )
   > n3 = &RigUnit_SphericalPoseReader(
    ActiveRegionScaleFactors=(PositiveWidth=1,NegativeWidth=1,PositiveHeight=1,NegativeHeight=1)
    ActiveRegionSize=0.12
    Debug=(bDrawDebug=false,bDraw2D=False,bDrawLocalAxes=False,DebugScale=25,DebugSegments=20,DebugThickness=0.25)
    DriverAxis=(X=0,Y=1,Z=0)
    DriverItem=(Type=Bone,Name="pelvis")
    FalloffRegionScaleFactors=(PositiveWidth=1,NegativeWidth=1,PositiveHeight=1,NegativeHeight=1)
    FalloffSize=0.2
    FlipHeightScaling=False
    FlipWidthScaling=False
    OptionalParentItem=(Type=Bone,Name="root")
    RotationOffset=(X=-15,Y=90,Z=0)
   )
   > RigVMVariableNode(Value=&RigVMFunction_MathDoubleRemap(
     SourceMaximum=0
     SourceMinimum=1
     TargetMaximum=0
     TargetMinimum=1
     Value=&RigVMFunction_MathDoubleMul(B=@n2.OutputParam, A=@n3.OutputParam).Result
    ).Result) {variable=Skydive Enabled}
   RigVMFunction_ControlFlowBranch = &RigVMFunction_ControlFlowBranch(ExecuteContext=@RigVMFunction_Sequence.D, Condition=RigVMVariableNode() {variable=Not Rigid Body}.Value)
   n4 = &RigVMFunction_Sequence(ExecuteContext=@RigVMFunction_ControlFlowBranch.True)
   RigVMFunction_MathQuaternionToEuler = &RigVMFunction_MathQuaternionToEuler(
    RotationOrder=ZYX
    Value=&RigUnit_GetRelativeTransformForItem(Child=(Type=Bone,Name="deform_head"), Parent=(Type=Bone,Name="spine_05"), bChildInitial=False, bParentInitial=False).RelativeTransform.Rotation
   )
   Clamp = &RigVMFunction_MathDoubleClamp(Maximum=30, Value=@RigVMFunction_MathQuaternionToEuler.Result.Y)
   n5 = &RigVMFunction_MathDoubleRemap(SourceMaximum=30, SourceMinimum=0, TargetMaximum=1, TargetMinimum=0, Value=@Clamp.Result)
   &RigUnit_ModifyTransforms(
    ItemToModify=((Item=(Type=Bone,Name="dyn_pipe_1_l"),Transform=(Rotation=(X=0,Y=0.2164,Z=0,W=0.9763),Translation=(X=0,Y=0,Z=5),Scale3D=(X=1,Y=1,Z=1))))
    WeightMinimum=0
    Weight=@n5.Result
    ExecutePin=@n4.A
   )
   > &RigUnit_ModifyTransforms(
    ItemToModify=((Item=(Type=Bone,Name="dyn_pipe_2_l"),Transform=(Rotation=(X=0,Y=-0.1305,Z=0,W=0.9914),Translation=(X=0,Y=0,Z=7),Scale3D=(X=1,Y=1,Z=1))))
    WeightMinimum=0
    Weight=@n5.Result
   )
   > &RigUnit_ModifyTransforms(
    ItemToModify=((Item=(Type=Bone,Name="dyn_pipe_3_l"),Transform=(Rotation=(X=0,Y=0,Z=0,W=1),Translation=(X=0,Y=0,Z=3),Scale3D=(X=1,Y=1,Z=1))))
    WeightMinimum=0
    Weight=@n5.Result
   )
   n6 = &RigVMFunction_MathDoubleRemap(SourceMaximum=30, SourceMinimum=0, TargetMaximum=1, TargetMinimum=0, Value=@RigVMFunction_MathQuaternionToEuler.Result.Z)
   > &RigUnit_ModifyTransforms(
    ItemToModify=((Item=(Type=Bone,Name="dyn_pipe_1_l"),Transform=(Rotation=(X=0,Y=0.2164,Z=0,W=0.9763),Translation=(X=0,Y=0,Z=3),Scale3D=(X=1,Y=1,Z=1))))
    WeightMinimum=0
    Weight=@n6.Result
   )
   > &RigUnit_ModifyTransforms(
    ItemToModify=((Item=(Type=Bone,Name="dyn_pipe_1_l"),Transform=(Rotation=(X=0,Y=0.2164,Z=0,W=0.9763),Translation=(X=0,Y=0,Z=1.3),Scale3D=(X=1,Y=1,Z=1))))
    WeightMinimum=0
    Weight=&RigVMFunction_MathDoubleRemap(SourceMaximum=-35, SourceMinimum=0, TargetMaximum=1, TargetMinimum=0, Value=@RigVMFunction_MathQuaternionToEuler.Result.X).Result
   )
   Negate = /Script/RigVM.RigVMFunction_MathDoubleNegate(Value=@RigVMFunction_MathQuaternionToEuler.Result.Y) {is=struct RigVMUnitNode, template="Negate::Execute(in Value,out Result)"}
   n7 = &RigVMFunction_MathDoubleRemap(
    SourceMaximum=30
    SourceMinimum=0
    TargetMaximum=1
    TargetMinimum=0
    Value=&RigVMFunction_MathDoubleClamp(Maximum=30, Value=@Negate.Result).Result
   )
   &RigUnit_ModifyTransforms(
    ItemToModify=((Item=(Type=Bone,Name="dyn_pipe_1_r"),Transform=(Rotation=(X=0,Y=0.2164,Z=0,W=0.9763),Translation=(X=0,Y=0,Z=-5),Scale3D=(X=1,Y=1,Z=1))))
    WeightMinimum=0
    Weight=@n7.Result
    ExecutePin=@n4.B
   )
   > &RigUnit_ModifyTransforms(
    ItemToModify=((Item=(Type=Bone,Name="dyn_pipe_2_r"),Transform=(Rotation=(X=0,Y=-0.1305,Z=0,W=0.9914),Translation=(X=0,Y=0,Z=-7),Scale3D=(X=1,Y=1,Z=1))))
    WeightMinimum=0
    Weight=@n7.Result
   )
   > &RigUnit_ModifyTransforms(
    ItemToModify=((Item=(Type=Bone,Name="dyn_pipe_3_r"),Transform=(Rotation=(X=0,Y=0,Z=0,W=1),Translation=(X=0,Y=0,Z=-3),Scale3D=(X=1,Y=1,Z=1))))
    WeightMinimum=0
    Weight=@n7.Result
   )
   > &RigUnit_ModifyTransforms(
    ItemToModify=((Item=(Type=Bone,Name="dyn_pipe_1_r"),Transform=(Rotation=(X=0,Y=0.2164,Z=0,W=0.9763),Translation=(X=0,Y=0,Z=-3),Scale3D=(X=1,Y=1,Z=1))))
    WeightMinimum=0
    Weight=@n6.Result
   )
   > &RigUnit_ModifyTransforms(
    ItemToModify=((Item=(Type=Bone,Name="dyn_pipe_1_r"),Transform=(Rotation=(X=0,Y=0.2164,Z=0,W=0.9763),Translation=(X=0,Y=0,Z=-1.3),Scale3D=(X=1,Y=1,Z=1))))
    WeightMinimum=0
    Weight=&RigVMFunction_MathDoubleRemap(SourceMaximum=35, SourceMinimum=0, TargetMaximum=1, TargetMinimum=0, Value=@RigVMFunction_MathQuaternionToEuler.Result.X).Result
   )
   Get Delta Ref = RigVMFunctionReferenceNode(Item=(Type=Bone,Name="clavicle_r"), ExecuteContext=@RigVMFunction_Sequence.E)
   > Get Delta Ref_1 = RigVMFunctionReferenceNode(Item=(Type=Bone,Name="clavicle_l"))
   Remap = &RigVMFunction_MathDoubleRemap(SourceMaximum=25, SourceMinimum=0, TargetMaximum=1, TargetMinimum=0, bClamp=False, Value=@"Get Delta Ref".Result.Rotation.Pitch)
   > &RigUnit_ModifyTransforms(
    ItemToModify=((Item=(Type=Bone,Name="deform_head"),Transform=(Rotation=(X=0,Y=0,Z=0,W=1),Translation=(X=0,Y=0,Z=0),Scale3D=(X=1,Y=1,Z=0.94))))
    WeightMinimum=0
    Weight=&RigVMFunction_MathDoubleMul(
     A=@Remap.Result
     B=&RigVMFunction_MathDoubleRemap(SourceMaximum=25, SourceMinimum=0, TargetMaximum=1, TargetMinimum=0, bClamp=False, Value=@"Get Delta Ref_1".Result.Rotation.Pitch).Result
    ).Result
   )
   n8 = &RigVMFunction_ControlFlowBranch(ExecuteContext=@RigVMFunction_Sequence.F, Condition=RigVMVariableNode() {variable=Lunch Emote}.Value)
   n9 = &RigVMFunction_AlphaInterp(
    InterpSpeedDecreasing=5
    InterpSpeedIncreasing=5
    Scale=1
    bClampResult=False
    Value=&RigVMFunction_MathDoubleRemap(
     SourceMaximum=53
     SourceMinimum=68
     TargetMaximum=1
     TargetMinimum=0
     Value=&RigUnit_GetRelativeTransformForItem(Child=(Type=Bone,Name="pelvis"), Parent=(Type=Bone,Name="root"), bChildInitial=False, bParentInitial=False).RelativeTransform.Translation.Z
    ).Result
   )
   &RigUnit_ModifyTransforms(
    ItemToModify=((Item=(Type=Bone,Name="deform_head"),Transform=(Rotation=(X=0,Y=0,Z=-0.2419,W=0.9703),Translation=(X=2.5,Y=-3,Z=0),Scale3D=(X=1,Y=1,Z=1))))
    WeightMinimum=0
    Weight=@n9.Result
    ExecutePin=@n8.True
   )
   n10 = &RigUnit_SphericalPoseReader(
    ActiveRegionScaleFactors=(PositiveWidth=1,NegativeWidth=1,PositiveHeight=1,NegativeHeight=1)
    ActiveRegionSize=0.07
    Debug=(bDrawDebug=false,bDraw2D=False,bDrawLocalAxes=False,DebugScale=25,DebugSegments=20,DebugThickness=0.25)
    DriverAxis=(X=1,Y=0,Z=0)
    DriverItem=(Type=Bone,Name="head")
    FalloffRegionScaleFactors=(PositiveWidth=1,NegativeWidth=1,PositiveHeight=1,NegativeHeight=1)
    FalloffSize=0.15
    FlipHeightScaling=False
    FlipWidthScaling=False
    OptionalParentItem=(Type=Bone,Name="spine_05")
    RotationOffset=(X=90,Y=0,Z=-142)
    ExecutePin=@RigVMFunction_Sequence.B
   )
   Multiply = &RigVMFunction_MathDoubleMul(A=@n10.OutputParam, B=RigVMVariableNode() {variable=Skydive Enabled}.Value)
   > &RigUnit_ModifyTransforms(
    ItemToModify=((Item=(Type=Bone,Name="neck_01"),Transform=(Rotation=(X=0,Y=0,Z=0.0872,W=0.9962),Translation=(X=0,Y=5,Z=0),Scale3D=(X=1,Y=1,Z=1))))
    WeightMinimum=0
    Weight=@Multiply.Result
   )
  ]}
]
hierarchy [
 root {loc=<0 0 0>, rot=<0 0 0 1>}
 pelvis {parent=root, loc=<0 -0.7321 92.4218>, rot=<0.0127 0.707 -0.0127 -0.707>}
 spine_01 {parent=pelvis, loc=<5.6002 -0.5128 0>, rot=<0 0 0.034 -0.9994>}
 spine_02 {parent=spine_01, loc=<8.0753 0.0139 0>, rot=<0 0 0.0238 -0.9997>}
 spine_03 {parent=spine_02, loc=<7.0393 0 0>, rot=<0 0 0.0149 -0.9999>}
 spine_04 {parent=spine_03, loc=<8.1024 0 0>, rot=<0 0 -0.0132 -0.9999>}
 spine_05 {parent=spine_04, loc=<10.0146 0 0>, rot=<0 0 -0.0015 -1>}
 attach_cape {parent=spine_05, loc=<9.1509 -13.3783 0>, rot=<0.0616 -0.7044 -0.0616 -0.7044>}
 attach_backpack {parent=spine_05, loc=<-0.123 -15.1423 0>, rot=<-0.0093 -0.707 0.0093 -0.707>}
 neck_01 {parent=spine_05, loc=<10.815 1.9196 0>, rot=<0 0 -0.1864 -0.9825>}
 neck_02 {parent=neck_01, loc=<4.7347 0.3045 0>, rot=<0 0 0.0363 -0.9993>}
 head {parent=neck_02, loc=<5.7632 0.5552 0>, rot=<0 0 0.115 -0.9934>}
 deform_head {parent=neck_02, loc=<5.7632 0.5552 0>, rot=<0 0 0.115 -0.9934>}
 dyn_pipe_1_r {parent=deform_head, loc=<-6.067 6.8707 14.4007>, rot=<-0.4816 0.507 0.4276 -0.5729>}
 dyn_pipe_2_r {parent=deform_head, loc=<-3.8286 -1.943 14.4186>, rot=<-0.5847 0.393 0.4311 -0.5638>}
 dyn_pipe_3_r {parent=deform_head, loc=<-1.0117 -9.7694 13.0183>, rot=<0.5972 -0.3561 -0.5166 0.4998>}
 dyn_pipe_1_l {parent=deform_head, loc=<-6.0672 6.8707 -14.4007>, rot=<0.507 0.4816 0.5729 0.4276>}
 dyn_pipe_2_l {parent=deform_head, loc=<-3.8283 -1.943 -14.4186>, rot=<-0.393 -0.5847 -0.5638 -0.4311>}
 dyn_pipe_3_l {parent=deform_head, loc=<-1.0115 -9.7694 -13.0183>, rot=<-0.3561 -0.5972 -0.4998 -0.5166>}
 clavicle_l {parent=spine_05, loc=<2.5856 -0.2866 -4.3777>, rot=<0.0465 -0.7289 -0.009 -0.6829>}
 upperarm_l {parent=clavicle_l, loc=<18.2264 0.1342 -0.001>, rot=<0.0253 -0.2693 0.0013 -0.9627>}
 upperarm_twist_01_l {parent=upperarm_l, loc=<7.9796 0.0843 -0.6195>, rot=<0 0 0 -1>}
 clavicle_r {parent=spine_05, loc=<2.5855 -0.2866 4.3777>, rot=<0.7289 0.0465 -0.6829 0.009>}
 upperarm_r {parent=clavicle_r, loc=<-18.2264 -0.1342 0.001>, rot=<0.0253 -0.2693 0.0013 -0.9627>}
 upperarm_twist_01_r {parent=upperarm_r, loc=<-7.9796 -0.0843 0.6195>, rot=<0 0 0 -1>}
]
variables [
 Front End {type=bool, default=False}
 Lunch Emote {type=bool, default=False}
 Not Rigid Body {type=bool, default=False}
 Skydive Enabled {type=double, default=0}
]
