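# Physics asset for the face accessory of the large male RadiumFox Flame character, covering six dynamic pipe bones, three per side.
# Each pipe capsule is constrained directly to deform_head rather than chained, with tight, equal swing and twist limit angles.
# Large head and clavicle capsules give volume around the pipes and join a tiny spine_05 capsule through constraints with free swing and twist.
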PhysicsAsset /BRCosmetics/Characters/Player/Male/Large/Bodies/M_LRG_RadiumFox_Flame/Meshes/Parts/M_LRG_RadiumFox_Flame_FaceAcc_Physics
bodies {
 clavicle_l [{
   shapeType=Capsule
   length=1.6737
   radius=8.5736
   shapeName=clavicle_l_capsule
   center=<8.0846 -0.2522 6.8881>
   rotation=<rot 70.6692 132.071 -45.9179>
  }]
 clavicle_r [{
   shapeType=Capsule
   length=1.6737
   radius=8.5736
   shapeName=clavicle_l_capsule
   center=<-8.708 0.9673 -6.9844>
   rotation=<rot 65.2452 108.4667 -71.7232>
  }]
 deform_head [{
   shapeType=Capsule
   length=1.952
   radius=10.1261
   shapeName=deform_head_capsule
   center=<3.1614 -0.0865 0.1987>
   rotation=<rot 0 36.8427 -89.3141>
  }]
 dyn_pipe_1_l [{
   shapeType=Capsule
   length=4.3827
   radius=1.9196
   shapeName=dyn_pipe_1_l_capsule
   center=<-0.1922 0.0772 0.0887>
   rotation=<rot 0 89.9372 -89.7362>
  }]
 dyn_pipe_1_r [{
   shapeType=Capsule
   length=4.106
   radius=1.9196
   shapeName=dyn_pipe_1_r_capsule
   center=<0.1922 -0.0772 -0.0885>
   rotation=<rot 0 89.9372 -89.7362>
  }]
 dyn_pipe_2_l [{
   shapeType=Capsule
   length=5.5237
   radius=1.8833
   shapeName=dyn_pipe_2_l_capsule
   center=<-0.3412 0.1557 0.0336>
   rotation=<rot 0 -91.2188 -89.6042>
  }]
 dyn_pipe_2_r [{
   shapeType=Capsule
   length=5.6513
   radius=1.8833
   shapeName=dyn_pipe_2_r_capsule
   center=<0.3413 -0.1557 -0.0339>
   rotation=<rot 0 -91.2188 -89.6042>
  }]
 dyn_pipe_3_l [{
   shapeType=Capsule
   length=3.021
   radius=1.8688
   shapeName=dyn_pipe_3_l_capsule
   center=<-0.2227 0.1678 -0.0553>
   rotation=<rot 0 -92.4305 -88.55>
  }]
 dyn_pipe_3_r [{
   shapeType=Capsule
   length=3.021
   radius=1.8688
   shapeName=dyn_pipe_3_r_capsule
   center=<0.2227 -0.1678 0.0552>
   rotation=<rot 0 -92.4305 -88.55>
  }]
 spine_05 [{shapeType=Capsule, length=0.505, radius=0.505, shapeName=spine_05_capsule, rotation=<rot 0 -90 0>}]
}
constraints [
 {
  bone1Name=clavicle_l
  bone2Name=spine_05
  swing1LimitDegrees=45
  swing1Motion=Free
  swing2LimitDegrees=45
  swing2Motion=Free
  twistLimitDegrees=45
  twistMotion=Free
 }
 {bone1Name=dyn_pipe_3_l, bone2Name=deform_head, swing1LimitDegrees=10, swing2LimitDegrees=10, twistLimitDegrees=10}
 {bone1Name=dyn_pipe_2_l, bone2Name=deform_head, swing1LimitDegrees=10, swing2LimitDegrees=10, twistLimitDegrees=10}
 {bone1Name=dyn_pipe_1_l, bone2Name=deform_head, swing1LimitDegrees=10, swing2LimitDegrees=10, twistLimitDegrees=10}
 {bone1Name=dyn_pipe_3_r, bone2Name=deform_head, swing1LimitDegrees=10, swing2LimitDegrees=10, twistLimitDegrees=10}
 {bone1Name=dyn_pipe_2_r, bone2Name=deform_head, swing1LimitDegrees=10, swing2LimitDegrees=10, twistLimitDegrees=10}
 {bone1Name=dyn_pipe_1_r, bone2Name=deform_head, swing1LimitDegrees=10, swing2LimitDegrees=10, twistLimitDegrees=10}
 {
  bone1Name=deform_head
  bone2Name=spine_05
  swing1LimitDegrees=45
  swing1Motion=Free
  swing2LimitDegrees=45
  swing2Motion=Free
  twistLimitDegrees=45
  twistMotion=Free
 }
 {
  bone1Name=clavicle_r
  bone2Name=spine_05
  swing1LimitDegrees=45
  swing1Motion=Free
  swing2LimitDegrees=45
  swing2Motion=Free
  twistLimitDegrees=45
  twistMotion=Free
 }
]
