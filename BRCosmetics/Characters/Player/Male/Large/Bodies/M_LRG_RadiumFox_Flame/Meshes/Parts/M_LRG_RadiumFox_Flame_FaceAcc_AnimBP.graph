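# Post-copy animation Blueprint for the face accessory of the large male Radium Fox Flame character cosmetic.
# It copies the attached parent's pose into a Control Rig, feeding it rigid-body, front-end and lunch-emote flags.
# A partially blended RigidBody node then adds physics simulated relative to the head bone, with overridden gravity.
# Each update sets the lunch-emote flag by comparing the owner mesh's active montage to a specific Builder emote.
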AnimBlueprint /BRCosmetics/Characters/Player/Male/Large/Bodies/M_LRG_RadiumFox_Flame/Meshes/Parts/M_LRG_RadiumFox_Flame_FaceAcc_AnimBP
target_skeleton=&/M_LRG_RadiumFox_Flame_FaceAcc_Skeleton
graphs [
 AnimGraph {nodes [
   n1 = Control Rig!(
    Source=Copy Pose From Mesh() {bUseAttachedParent=true}
    Alpha
    Not Rigid Body:bool=NOT Boolean(A=Get bRigidBodyEnabled())
    Front End:bool=Get bIsInFrontEnd()
    Lunch Emote:bool=Get Lunch Emote() -> (Lunch Emote:bool)
   ) -> (Pose) {
    alphaBoolBlend {blendOption=Linear}
    controlRigAssetReference {blueprintRigClass=&/M_LRG_RadiumFox_Flame_FaceAcc_CtrlRig.M_LRG_RadiumFox_Flame_FaceAcc_CtrlRig_C}
   }
   n2 = RigidBody(ComponentPose=Local To Component(LocalPose=@n1), Alpha=0.65) {
    bOverrideWorldGravity=true
    overridePhysicsAsset=&/M_LRG_RadiumFox_Flame_FaceAcc_Physics
    simulationSpace=BaseBoneSpace
    alphaBoolBlend {blendOption=Linear}
    baseBoneRef {boneName=deform_head}
    componentAppliedLinearAccClamp=<5000 5000 5000>
    componentLinearAccScale=<0.7 0.7 0.7>
    componentLinearVelScale=<0.7 0.7 0.7>
    overrideWorldGravity {z=-980}
    simSpaceSettings {maxAngularAcceleration=450, maxAngularVelocity=35, maxLinearAcceleration=150, maxLinearVelocity=10, velocityScaleZ=0, worldAlpha=0.8}
   }
   Output Pose(Result=Component To Local(ComponentPose=@n2)) {layerGroup=None}
  ]}
 EventGraph {nodes [
   Event BlueprintUpdateAnimation()
   n1 = GetAnimInstance(self=GetOwnerSkeletalMeshComponent())
   > Set Lunch Emote(Lunch Emote:bool="Equal (Object)"(
     A=GetCurrentActiveMontage!(self:Anim Instance Object Reference=@n1) -> (ReturnValue:Anim Montage Object Reference)
     B=/BRCosmetics/Animation/Game/MainPlayer/Emotes/Builder/Emote_Builder_CMM_M
    )) -> (Output_Get:bool)
  ]}
]
variables [Lunch Emote {type=boolean}]
